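# Idle cosmetic Niagara effect that spawns short-lived red-to-pink sparkles on the FloraBrisk character's body.
# One emitter's particle partition selects both a skeletal bone/socket and a baked array of offset points; nine arrays exist, but only three partitions are configured.
# Curves fade and scale the two-renderer flare sprites, modulated by quality level and time of day, with front-end rate overrides.
# The foot-hiding kill step is disabled.
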NiagaraSystem /BRCosmetics/Characters/Player/Female/Medium/Bodies/F_MED_FloraBrisk/FX/NS_Character_FloraBrisk_Idle
dependencies {
 usedDataInterfaces [
  &NiagaraDataInterfaceArrayFloat3
  &NiagaraDataInterfaceCurve
  &NiagaraDataInterfaceArrayFloat
  &NiagaraDataInterfaceSkeletalMesh
  &NiagaraDataInterfaceColorCurve
 ]
 usedDynamicInputs [
  &Multiply_Vector2D_ByFloat
  &Lerp_Float
  &MakeVector2D
 ]
 usedModules [
  &self:SystemSpawnScript.SystemScriptSource.SystemScriptGraph.NiagaraNodeAssignment_0.SetVariables_#1
  &SystemState
  /BRCosmetics/Characters/Player/Female/Medium/Bodies/F_MED_PiperShelf_Pearl/FX/TimeOfDayGlobal
 ]
}
emitters {Sparkles_Original {
  emitterUpdateScript [
   &EmitterState(Life Cycle Mode=::NewEnumerator1, Loop Behavior=::NewEnumerator1, Loop Duration Mode=::NewEnumerator1, UseLoopDelay=false)
   &SpawnRate(
    Spawn Probability=0.6
    SpawnRate=&FN_FrontEndFloatModifier(Gameplay=15, Front End=12)
    Use Spawn Probability=true
   )
  ]
  particleSpawnScript [
   &PartitionParticles(Number Of Partitions=3)
   &InitializeParticle(
    Color=&RandomRangeLinearColor(Minimum=<rgba 2 0 0 1>, Maximum=<rgba 2 0.4792 0.4792 1>)
    Lifetime Max=1
    Lifetime Min=0.5
    Lifetime Mode=::NewEnumerator1
    Sprite Size Mode=::NewEnumerator1
    Uniform Sprite Size Max=12
    Uniform Sprite Size Min=10
    UsePositionOffset=false
    Write Material Random=true
   )
   &CameraOffset(Camera Offset Amount=&FN_FrontEndFloatModifier(Gameplay=5, Front End=5))
  ]
  particleUpdateScript [
   &ParticleState()
   &SkeletalMeshLocation(
    Bone / Socket Index=$Particles.PartitionParticles.Partition
    Bone Sampling Mode=::NewEnumerator4
    MaskSpawnGroup=false
    Sampled Position Offset=&Multiply_VectorByFloat(
     Vector=&MergeHelper_VectorSwitch(
      0=&SelectVectorFromArray(
       Array Sampling Mode=::NewEnumerator0
       Direct Array Index=?
       Vector Selection Array=DataInterface() {propertyValues {FloatData [
          <-6.6316 -0.2471 -16.4935>
          <-5.5657 4.969 -13.2785>
          <-5.3927 5.3064 -12.7274>
          <-8.992 6.6269 -7.4848>
          <-9.5238 4.8928 -3.4497>
          <-12.0412 2.6632 0.8663>
          <-10.8491 4.2527 -0.7294>
          <-7.2814 6.3523 0.6272>
          <-6.3493 6.7399 -0.4067>
          <-6.5097 6.5472 -2.1457>
          <-5.792 6.63 -3.7167>
          <-6.5735 6.4293 -8.1514>
          <-4.5735 6.756 -4.8894>
          <-2.9966 7.6017 1.1076>
          <-2.0272 7.6954 -1.8396>
          <-7.082 6.1654 3.4361>
          <-7.498 6.3412 7.1017>
          <-5.3149 6.217 10.018>
          <-7.9549 6.0037 12.4532>
          <-3.4594 1.0659 15.3132>
          <3.2436 -1.1038 13.9841>
          <1.3461 -0.0098 14.4785>
          <-2.7037 6.3188 8.5135>
          <-5.097 6.585 5.2503>
          <-3.925 7.0088 4.2858>
          <0.8643 6.9493 6.6764>
          <2.1512 6.7599 7.1487>
          <2.4256 7.5149 4.3919>
          <0.3802 7.9229 1.5197>
          <2.4085 7.9071 -2.2441>
          <0.4364 7.4232 -4.8624>
          <-1.2022 7.4494 -4.195>
          <0.0437 6.9386 -6.6149>
          <1.1134 6.072 -9.3636>
          <1.3809 0.8605 -14.1898>
          <1.2126 6.4899 -8.1407>
          <0.933 -7.1426 -13.553>
          <1.215 -11.0511 -10.2394>
          <0.7423 -12.9369 -7.2229>
          <-4.0355 -4.9094 15.6763>
          <0.8459 -8.9437 12.5326>
          <1.0796 -10.1325 11.3893>
          <-0.4802 -13.9943 4.2861>
          <-0.8711 -14.2866 1.6623>
          <2.5488 -13.2454 -0.0958>
          <1.1961 -13.2532 -5.3738>
          <-3.4711 -14.6475 -1.0494>
          <-4.3227 -14.249 -6.1749>
          <-1.8985 -14.5078 1.6764>
          <-2.7953 -14.5533 2.5418>
          <-4.838 -14.5942 0.9485>
          <-4.8692 -14.4842 -4.3745>
          <-1.7361 -14.1302 5.3248>
          <-7.9896 -13.9092 1.4147>
          <-3.6731 -12.4632 10.5584>
          <-6.1529 -13.796 8.1235>
          <-5.1464 -11.5722 12.0328>
          <-7.2202 -10.7794 12.9531>
          <-7.1708 -12.9502 10.0391>
          <-7.7457 -13.2318 9.108>
          <-8.9256 -11.497 11.899>
          <-11.9415 -11.2321 2.1165>
          <-10.0049 -12.7052 -0.1259>
          <-8.2485 -13.7882 -0.764>
          <-10.2329 -13.0973 -5.0545>
          <-4.8909 -12.8656 -10.1746>
          <-8.3681 -12.8057 -10.0316>
          <-8.0492 -11.4473 -12.2076>
          <-4.4676 -9.727 -13.5561>
          <-7.2939 -8.2419 -14.9131>
          <-7.4216 -4.3282 -16.587>
          <-9.9588 -10.2797 -13.1599>
          <-11.6976 -12.0819 -8.4893>
          <-12.4041 -11.7398 -5.6675>
          <-13.3056 -10.0745 -3.533>
          <-11.3426 -12.3361 -4.4399>
          <-12.616 -10.4249 1.9356>
          <-14.2582 -6.9362 -2.2299>
          <-12.4973 -10.1881 -0.7136>
          <-13.803 -6.4161 0.536>
          <-13.2206 -8.5166 0.556>
          <-14.2062 -8.4708 3.0965>
          <-13.3843 -9.9013 3.3884>
          <-13.2723 -11.2035 7.8727>
          <-14.1339 -10.51 7.837>
          <-14.2955 -10.4395 8.9614>
          <-12.4142 -10.8473 10.9444>
          <-7.6738 -4.6231 16.5651>
         ]}}
       Random Seed=0
      )
      1=&SelectVectorFromArray(
       Array Sampling Mode=::NewEnumerator0
       Direct Array Index=?
       Vector Selection Array=DataInterface() {propertyValues {FloatData [
          <-8.0043 -2.2335 13.4572>
          <-7.2767 -0.4251 13.0514>
          <-5.267 -1.6612 12.0795>
          <-8.2158 3.9205 11.4731>
          <-9.2683 3.8801 11.8952>
          <-5.0104 -1.5531 11.9454>
          <-6.1295 3.2732 11.0909>
          <-7.6935 4.3923 10.9135>
          <-1.2431 -0.3992 10.4187>
          <-0.2156 3.4392 9.5077>
          <-3.4735 6.0303 7.8968>
          <3.9013 -0.3204 10.8847>
          <-9.9691 6.9312 7.5317>
          <-8.1371 6.9508 7.3096>
          <-7.9359 6.5369 8.106>
          <-9.8151 7.881 4.7891>
          <-9.2939 8.4263 2.0765>
          <-6.0063 8.3484 3.3782>
          <-5.4205 8.4389 3.1223>
          <-3.4704 7.5388 5.8829>
          <-2.8281 7.4396 6.1418>
          <-8.0402 4.6477 -10.8539>
          <-8.6542 3.6098 -11.8795>
          <-6.7703 -0.2197 -12.7872>
          <-5.2602 -0.8518 -12.0507>
          <-5.6073 3.9885 -10.3564>
          <-3.9824 -0.4338 -11.3683>
          <-0.4316 3.4269 -9.4931>
          <4.4577 0.7142 -10.9716>
          <5.1338 1.5351 -10.9837>
          <5.2825 3.1852 -10.4762>
          <6.0641 3.2997 -10.5542>
          <3.8285 3.4305 -10.1273>
          <3.3196 3.3872 -10.074>
          <5.54 6.391 -8.5696>
          <5.4721 10.1647 -2.9079>
          <4.6875 9.2353 -5.3235>
          <2.4837 9.1578 -4.7779>
          <1.2707 5.0148 -8.9605>
          <0.4713 6.8568 -7.4845>
          <0.5543 7.9322 -6.3104>
          <0.3464 8.4069 -5.3593>
          <-4.0186 2.8932 -10.4376>
          <-5.0944 4.9887 -9.3581>
          <-7.6754 5.3362 -9.8158>
          <-7.3947 6.6649 -7.6976>
          <-9.1438 7.0453 -7.191>
          <-9.1491 7.5122 -5.9031>
          <-4.275 8.8286 -0.778>
          <-3.0975 8.6955 -2.9829>
          <-3.1993 9.0002 -0.3177>
          <-0.137 9.3962 -2.1565>
          <-2.7607 8.5022 3.9408>
          <-1.9624 8.4453 4.4148>
          <-1.7249 9.1197 2.1372>
          <0.3659 9.0553 3.9246>
          <-0.0191 8.0628 5.8882>
          <1.2964 9.7889 -0.0639>
          <3.5726 10.0966 -1.5386>
          <4.5801 10.2393 1.9642>
          <1.4062 9.2935 3.7861>
          <5.3181 10.3536 0.9728>
          <6.0008 3.0516 10.6291>
          <6.8944 4.2697 10.1276>
          <4.2954 2.5976 10.5097>
          <5.2565 9.3882 5.1271>
          <5.1593 -4.7778 -9.3181>
          <3.9249 -2.8349 -10.3474>
          <-3.0328 -2.9921 -10.4644>
          <-3.7654 -2.3997 -11.0698>
          <-6.4282 -2.512 -12.6345>
          <0.3874 -4.8743 -8.3092>
          <-8.1128 -3.2342 -13.3396>
          <-8.113 -4.4129 -13.0567>
          <-7.3184 -6.8196 -11.4472>
          <-5.7961 -6.4005 -10.3992>
          <-6.3702 -8.6508 -8.8014>
          <-8.1072 -10.9076 -7.6476>
          <-4.751 -9.0602 -5.3228>
          <2.9103 -5.6814 -7.6413>
          <5.0004 -7.6743 -3.309>
          <5.9872 -3.2039 10.6848>
          <2.9591 -4.3711 9.0939>
          <1.4602 -1.2957 10.3547>
          <-1.0459 -1.3137 10.3675>
          <-4.8191 -4.9328 10.5671>
          <-6.1152 -5.6828 11.1631>
          <-7.1596 -5.1624 12.2505>
          <-6.1821 -6.8082 10.4881>
          <-6.4973 -8.065 9.6459>
          <-5.4574 -7.8936 8.5619>
          <-0.6781 -6.1335 7.1249>
          <-1.6584 -6.4449 7.2107>
          <-6.8191 -10.5959 6.0073>
          <-3.7554 -7.7866 6.6516>
          <-2.3409 -7.3061 5.8809>
          <-5.0968 -9.3789 5.2486>
          <-3.4003 -8.1979 5.0309>
          <-8.1043 -12.1908 2.9137>
          <-7.7468 -12.0788 0.9559>
          <-3.4667 -8.7676 2.5052>
          <-1.1414 -7.065 5.16>
          <0.0186 -7.0031 4.911>
          <5.4081 -6.2028 7.5498>
          <6.4122 -6.7334 6.9552>
          <2.9278 -7.1462 4.5881>
          <4.6876 -7.8167 1.5561>
          <3.0288 -7.6503 0.673>
          <1.4001 -7.4791 1.8128>
          <-1.1625 -7.6111 1.542>
          <0.2851 -7.5475 -0.0122>
          <0.9756 -7.4135 -2.3316>
          <-5.1492 -10.1637 -0.6612>
          <-4.8319 -9.6954 -2.8799>
          <-6.6303 -11.0085 -3.9118>
         ]}}
       Random Seed=0
      )
      2=&SelectVectorFromArray(
       Array Sampling Mode=::NewEnumerator0
       Direct Array Index=?
       Vector Selection Array=DataInterface() {propertyValues {FloatData [
          <-5.5236 -0.0459 11.6477>
          <-4.3709 -1.1972 11.9291>
          <-2.3703 -1.3471 12.278>
          <-0.5057 -1.2081 12.6236>
          <-3.8522 0.5575 -11.9629>
          <-3.5031 2.1948 -11.6767>
          <-4.3635 2.8071 -11.2898>
          <-0.7525 -0.5551 -12.5563>
          <-0.3664 -1.6233 -12.6565>
          <-2.3297 2.1432 -11.8447>
          <1.2988 -1.6476 -12.6506>
          <4.3025 3.4605 -12.3468>
          <-0.0589 3.672 -11.8914>
          <3.416 5.2802 -12.0465>
          <4.6735 4.175 -12.3692>
          <1.8397 5.9274 -11.7244>
          <7.2052 7.3853 -9.1628>
          <6.7668 7.9529 -8.7421>
          <6.2432 9.0132 -2.7869>
          <5.6268 9.6194 -6.1858>
          <6.0621 9.4772 -4.1502>
          <5.6047 9.763 -5.113>
          <5.6546 9.5746 -2.2988>
          <5.5989 9.6026 -6.369>
          <5.5898 9.7765 -3.7105>
          <4.8077 8.3085 -9.7715>
          <3.1648 10.972 -5.0997>
          <3.8953 9.0951 -9.2747>
          <1.6821 11.3568 -4.9329>
          <2.088 10.512 -7.7395>
          <3.3954 10.8402 -2.8247>
          <3.0557 10.9233 -2.0588>
          <2.4377 11.213 -2.8159>
          <5.8254 9.576 3.2524>
          <5.1272 9.9762 2.8361>
          <6.4396 9.2351 4.9669>
          <5.1829 10.0203 4.9872>
          <5.5013 9.8287 3.7902>
          <4.7754 9.9549 7.0809>
          <6.95 7.9373 8.5313>
          <6.2893 8.6684 7.9577>
          <6.8523 7.8114 8.8676>
          <5.3063 8.9303 8.5669>
          <4.997 7.3115 10.6936>
          <5.7868 7.9001 9.7088>
          <4.5481 3.746 12.3687>
          <3.4836 7.3786 11.0556>
          <3.352 4.6968 12.1985>
          <2.2806 5.8974 11.789>
          <1.9655 2.8333 12.3584>
          <0.8831 -0.4988 12.7355>
          <0.6085 0.1897 12.6452>
          <-0.5867 0.5784 12.4341>
          <0.3992 2.5034 12.1845>
          <-0.4835 3.1874 11.8731>
          <0.4527 5.3064 11.649>
          <2.555 9.6705 9.0542>
          <1.1044 9.8116 9.0049>
          <0.2507 10.0401 8.5701>
          <-1.2537 9.6145 8.5504>
          <3.2128 10.4741 7.3167>
          <4.0416 10.6123 5.2167>
          <1.1823 11.1851 6.3241>
          <2.909 11.1184 3.8262>
          <0.3443 11.3582 5.6489>
          <2.5158 11.2146 3.1486>
          <2.2599 11.1913 -1.7803>
          <1.5197 11.439 -2.6871>
          <-0.321 11.6149 3.551>
          <0.9562 11.4774 -1.4564>
          <0.5197 11.5554 -2.712>
          <0.1052 11.2301 -6.0623>
          <1.5168 9.1017 -9.8224>
          <0.4334 5.2124 -11.6634>
          <-1.49 4.391 -11.3069>
          <-2.1133 5.4761 -10.6829>
          <-4.4411 4.2966 -10.6331>
          <-1.9902 9.3428 -8.359>
          <-1.9276 10.4617 -6.8849>
          <-3.8014 7.8993 -8.4651>
          <-3.9824 9.1109 -7.0653>
          <-4.16 6.943 -8.9861>
          <-3.6684 7.0365 -9.1849>
          <-5.3113 5.2934 -9.8829>
          <-5.7376 7.6899 -7.9858>
          <-4.6842 8.7045 -7.1598>
          <-5.1109 9.3875 -6.0745>
          <-5.8145 9.943 -4.9735>
          <-3.7796 10.3452 -5.1957>
          <-6.627 10.5087 -3.0088>
          <-5.1572 10.8137 -1.6495>
          <-4.3002 10.8912 1.8018>
          <-4.2757 10.9078 0.2979>
          <-3.0913 10.817 4.5074>
          <-2.221 11.1388 4.2235>
          <-0.4455 11.5946 2.9055>
          <-1.3678 11.3964 4.088>
          <-4.3097 7.8148 8.2893>
          <-5.6137 7.9578 7.7377>
          <-5.4448 8.9069 6.6995>
          <-2.7917 5.4634 10.4855>
          <-2.9523 3.9355 11.084>
          <-3.8201 4.4678 10.6793>
          <-4.2934 5.4049 10.0845>
          <-5.0874 2.3503 11.2813>
          <1.0818 -2.1826 -12.6543>
          <-0.6664 -7.3333 -8.0953>
          <-0.9276 -6.9513 -8.8522>
          <-2.4556 -7.5807 -6.6739>
          <-3.6551 -6.8722 -7.6573>
          <-2.954 -6.9778 -7.7933>
          <-1.425 -8.4864 -1.6853>
          <-3.3413 -5.9048 -9.3731>
          <-3.0353 -4.6721 -10.773>
          <-3.8819 -2.763 -11.604>
          <-5.3791 -4.4118 -10.2486>
          <-5.8042 -3.6941 -10.6416>
          <-5.4944 -2.1008 -11.4173>
          <-4.7282 -6.6487 -7.6006>
          <-5.7947 -6.5037 -7.4313>
          <-4.0214 -7.9411 -3.8239>
          <-4.7127 -7.9665 -2.6289>
          <-5.1359 -8.0203 -0.6136>
          <-4.087 -8.1223 -1.3111>
          <-4.2585 -8.1084 1.0505>
          <-1.7211 -8.4512 1.2396>
          <-1.4499 -8.443 2.509>
          <-3.1872 -8.1621 3.0374>
          <-1.7066 -7.6735 6.8684>
          <-0.6756 -7.7249 7.2671>
          <0.2204 -2.0741 12.663>
          <-0.7009 -2.9566 12.2459>
          <-2.4589 -3.8967 11.5428>
          <-4.1753 -3.6386 11.1789>
          <-3.2365 -5.5919 9.8243>
          <-4.3556 -4.5205 10.4628>
          <-5.7231 -4.9858 9.6536>
          <-4.9958 -3.2426 11.1888>
          <-5.5759 -1.2608 11.644>
          <-3.521 -6.9612 7.5564>
          <-3.8148 -7.6262 5.5934>
          <-5.2075 -6.3319 8.0174>
          <-4.3825 -7.3126 6.2309>
          <-5.5005 -7.503 5.0438>
         ]}}
       Random Seed=0
      )
      3=&SelectVectorFromArray(
       Array Sampling Mode=::NewEnumerator0
       Direct Array Index=?
       Vector Selection Array=DataInterface() {propertyValues {FloatData [
          <46.6857 1.5682 -4.2413>
          <44.955 -1.3961 -3.1683>
          <41.852 -1.6206 -4.0959>
          <40.9942 -0.6589 -5.0857>
          <35.7501 1.2448 -6.1402>
          <33.0805 -1.1165 -6.1696>
          <27.2912 -1.2874 -6.8622>
          <36.0311 -3.3426 -4.3061>
          <31.2167 -5.2222 -3.0483>
          <24.2743 -2.4158 -6.681>
          <28.8445 -4.1446 -4.9018>
          <22.9329 -4.1131 -5.7569>
          <19.5775 -2.0893 -7.4633>
          <23.7805 -4.9914 -4.6206>
          <13.0634 -3.2806 -6.8101>
          <9.83 -3.3721 -6.6493>
          <11.0823 -4.6766 -5.4648>
          <8.5272 -5.4621 -4.076>
          <13.29 -4.5918 -5.7872>
          <10.8491 -5.6949 -3.902>
          <11.6702 -6.7242 -1.2898>
          <17.0088 -6.6907 -1.9391>
          <19.1965 -5.6747 -4.0468>
          <22.2166 -5.4457 -4.1213>
          <25.0856 -6.339 -2.3131>
          <19.096 -6.9418 0.8765>
          <28.668 -6.4577 0.1821>
          <32.2994 -6.0746 0.4402>
          <29.7801 -5.5676 2.8197>
          <33.979 -5.6619 1.2663>
          <35.2547 -4.7098 3.0378>
          <39.816 -4.8245 0.74>
          <39.8913 -3.735 3.1721>
          <48.6551 5.7723 -2.9431>
          <46.0587 4.4418 -3.4806>
          <41.6378 3.9214 -4.2925>
          <43.046 2.3642 -5.1299>
          <42.9084 4.8569 -2.7566>
          <39.5773 4.2736 -4.2772>
          <37.7715 2.7823 -5.5008>
          <39.7305 5.4824 -2.1139>
          <37.8304 5.9125 -1.4324>
          <29.1316 3.6232 -6.2518>
          <33.3515 4.3605 -5.1758>
          <31.3756 5.4015 -4.5184>
          <34.0054 5.3725 -3.9099>
          <22.3762 5.0326 -6.4286>
          <21.8083 6.5445 -5.2983>
          <15.4447 1.61 -8.4471>
          <18.9313 7.0035 -5.3585>
          <20.7024 8.3108 -3.1286>
          <15.9572 9.8202 -0.3742>
          <21.2748 8.9178 -1.0569>
          <13.4383 5.0401 -7.4167>
          <14.7305 7.0451 -6.0319>
          <14.5338 9.9294 -1.7857>
          <19.1495 9.1568 1.2649>
          <8.5254 0.7767 7.547>
          <13.8621 5.3354 7.0352>
          <13.0868 -0.4001 7.4696>
          <11.0146 -1.7924 6.9124>
          <13.2655 -3.0572 6.4527>
          <7.8809 -6.2677 2.4098>
          <8.7652 -6.8536 0.7741>
          <10.8483 -6.5246 2.3665>
          <12.021 -5.497 4.4283>
          <13.8627 -5.9884 3.9754>
          <17.9296 -6.6568 1.8837>
          <14.1717 -3.6713 6.1794>
          <14.2055 -0.9292 7.2971>
          <17.2691 1.5593 7.4375>
          <24.0908 -3.839 5.3528>
          <20.966 2.0472 7.0891>
          <22.4432 -0.3326 6.8325>
          <26.327 2.0798 6.3099>
          <39.8179 -0.1859 5.1204>
          <47.0522 6.8044 -0.7958>
          <45.3213 6.252 2.4679>
          <43.7802 5.9324 -0.6023>
          <43.783 6.0767 -0.1734>
          <38.6038 5.3384 3.0474>
          <36.4868 4.2343 4.3757>
          <34.3138 2.2202 5.4093>
          <39.2935 6.0335 -0.6149>
          <29.7865 2.6312 5.7058>
          <28.6926 7.6137 -1.2488>
          <33.6551 6.3102 2.3079>
          <29.1399 5.1856 4.7027>
          <31.6322 7.1903 0.7003>
          <30.3571 7.41 0.8461>
          <30.1956 7.2621 1.7802>
          <27.6409 7.879 1.0835>
          <27.2099 7.863 -1.3278>
          <23.1048 8.5625 -1.3014>
          <23.5423 8.3572 2.0477>
          <23.2543 3.643 6.3608>
          <21.3683 3.4943 6.711>
          <23.5602 7.3209 4.0016>
          <18.7701 3.6675 7.0104>
          <19.2649 6.7286 5.3087>
          <17.2652 6.8521 5.5274>
          <14.576 6.5468 6.1825>
         ]}}
       Random Seed=0
      )
      4=&SelectVectorFromArray(
       Array Sampling Mode=::NewEnumerator0
       Direct Array Index=?
       Vector Selection Array=DataInterface() {propertyValues {FloatData [
          <36.9771 5.1469 -1.1255>
          <32.2241 5.2091 -1.8758>
          <34.7374 6.0254 1.7417>
          <37.439 5.455 2.1076>
          <41.4736 2.8968 3.0879>
          <39.5847 1.0127 3.1693>
          <43.5435 0.1852 3.3948>
          <42.7205 -1.1337 3.1081>
          <38.7249 -0.8148 -0.6734>
          <36.7666 1.9143 -2.1028>
          <30.2592 2.3428 -2.6499>
          <32.9903 -0.7276 -0.478>
          <25.7605 2.0682 -3.2173>
          <21.039 2.2074 -3.9314>
          <27.7553 -0.8896 -0.77>
          <18.733 1.4452 -3.7816>
          <19.1267 0.0514 -2.8241>
          <23.3634 -1.4426 -0.6767>
          <18.0208 -0.6957 -2.3447>
          <15.2496 -0.0881 -3.1089>
          <6.2106 -0.2076 -3.8188>
          <9.8272 -2.3595 -1.2195>
          <3.5695 -2.6746 -2.2743>
          <3.1592 -3.9333 0.7793>
          <7.9086 -3.0611 1.4262>
          <9.6679 -3.1271 1.345>
          <16.087 -2.7202 0.9675>
          <20.8572 -2.1188 0.2695>
          <19.6129 -1.5532 3.3547>
          <23.9367 -1.8885 0.824>
          <22.3593 -1.0264 3.7492>
          <26.2794 -1.0323 2.8958>
          <34.7795 5.3029 2.7996>
          <28.2909 6.9544 2.5018>
          <26.5798 5.9367 4.2989>
          <23.2291 5.8377 5.0986>
          <24.1211 2.8029 5.4265>
          <22.3694 6.2958 4.8445>
          <21.6901 8.4956 1.3873>
          <21.3875 8.4425 0.7431>
          <24.1299 7.3821 -1.4358>
          <27.3718 3.7445 -3.1858>
          <24.8847 5.2069 -3.258>
          <24.0354 4.5847 -3.7259>
          <20.453 6.3932 -3.5519>
          <20.4777 7.9583 -1.7375>
          <20.0399 7.9669 3.5196>
          <15.5589 8.4415 -1.6727>
          <12.6986 4.8725 -4.495>
          <12.5404 7.525 -2.5745>
          <4.5613 -1.2873 3.5368>
          <13.6486 -0.3035 4.6617>
          <14.2276 1.2601 5.5064>
          <22.6741 -0.47 4.1326>
          <23.84 0.2281 4.4489>
          <19.3566 1.6274 5.5418>
          <21.8159 3.7862 5.7749>
          <9.4421 6.1573 4.4421>
          <10.1301 8.0417 -0.2787>
          <7.3264 7.4439 1.0258>
          <5.3991 5.2631 3.9645>
         ]}}
       Random Seed=0
      )
      5=&SelectVectorFromArray(
       Array Sampling Mode=::NewEnumerator0
       Direct Array Index=?
       Vector Selection Array=DataInterface() {propertyValues {FloatData [
          <-2.2584 -1.0125 2.8205>
          <-2.6499 -2.7121 -2.8343>
          <6.8401 -3.711 4.157>
          <8.1345 -4.5175 4.3991>
          <7.3576 -5.61 4.6398>
          <8.0407 -10.3286 3.5654>
          <7.7537 -12.8673 2.2462>
          <7.6612 -11.9083 -1.7254>
          <5.8185 -10.2544 -2.509>
          <6.5756 -7.908 -3.6305>
          <-6.8173 3.5479 -2.1995>
          <-3.0127 0.961 -3.3367>
          <-1.5315 -0.8019 -3.2102>
          <-1.4949 -1.3353 -3.0021>
          <2.0986 -1.7689 -2.9503>
          <0.4655 -7.0729 -1.2642>
          <0.7541 -7.3462 -1.085>
          <3.8402 -12.5365 -0.4545>
          <3.4698 -12.6714 0.5533>
          <4.2425 -13.1514 0.4385>
          <4.3083 -13.1653 0.3988>
          <-2.2645 -3.9163 2.7058>
          <-1.1153 -6.0118 1.7647>
          <-1.699 -4.465 2.3268>
          <1.4089 -4.9314 3.5114>
          <-0.3934 2.8506 2.4902>
          <4.6731 -5.358 4.5969>
          <-2.6904 1.516 2.9467>
          <-4.373 3.9823 2.0935>
          <8.224 -6.9304 2.6401>
          <8.1533 -5.9378 -2.651>
          <8.1231 -8.3578 -3.3337>
          <8.2098 -4.3634 3.5501>
          <6.8164 -3.6613 -0.0839>
          <2.7993 -1.7479 0.4945>
          <1.6633 -0.7403 -1.0679>
          <1.2985 0.1158 0.9377>
          <2.2138 0.9854 1.061>
          <8.3413 2.2319 -0.1598>
          <8.3435 2.5703 -1.1279>
          <-0.4486 3.7054 -2.6016>
          <2.3282 3.4781 1.453>
          <4.1976 2.9574 1.286>
          <3.1291 3.5504 1.1199>
          <5.5394 3.4317 -0.8865>
          <6.2112 2.1819 -1.7518>
         ]}}
       Random Seed=0
      )
      6=&SelectVectorFromArray(
       Array Sampling Mode=::NewEnumerator0
       Direct Array Index=?
       Vector Selection Array=DataInterface() {propertyValues {FloatData [
          <-44.0372 -2.1795 4.9005>
          <-43.1923 1.8055 3.5152>
          <-38.4367 1.3701 5.0704>
          <-34.4797 0.8514 6.0711>
          <-31.8519 -1.202 6.5943>
          <-32.1049 0.5655 6.3766>
          <-33.5728 2.9903 5.0091>
          <-31.295 2.4206 5.6813>
          <-23.7986 3.4613 6.0794>
          <-14.9768 -0.1199 8.4179>
          <-17.8289 3.7766 6.3755>
          <-14.2915 4.107 6.1901>
          <-12.9217 5.9958 3.6015>
          <-11.6818 6.0116 3.4841>
          <-9.0261 6.1391 2.8426>
          <-17.9586 5.4546 4.461>
          <-14.0851 6.828 1.3224>
          <-21.4157 4.7268 5.2537>
          <-17.5926 6.828 1.1566>
          <-18.1997 6.7068 1.7236>
          <-23.7514 5.7895 3.531>
          <-24.1624 5.3629 4.0486>
          <-23.506 6.4102 2.4561>
          <-25.7438 6.5066 1.1677>
          <-26.7976 6.2833 1.9925>
          <-32.5379 4.9892 3.0561>
          <-36.1356 3.8166 3.6591>
          <-37.8444 3.7901 3.1273>
          <-45.142 3.5328 -1.9855>
          <-43.8531 3.7151 -2.2256>
          <-41.8342 4.7029 -1.2024>
          <-39.7948 4.7824 -1.2634>
          <-40.072 1.7425 -4.5483>
          <-37.3346 4.2583 -3.1265>
          <-38.0038 -0.1006 -5.261>
          <-34.8137 3.5234 -4.225>
          <-33.2491 2.9306 -4.992>
          <-39.5795 -2.9783 -5.0128>
          <-36.5457 -1.8768 -5.4188>
          <-40.2791 -3.7551 -4.6598>
          <-44.127 -2.5443 -5.2755>
          <-45.0901 -2.6277 -5.3294>
          <-45.3519 -5.4104 -3.9481>
          <-42.0371 -4.3142 3.8544>
          <-39.985 -2.6246 5.3511>
          <-37.8302 -3.0619 5.3425>
          <-38.2116 -6.048 -1.4592>
          <-44.205 -6.1443 -1.9098>
          <-36.2199 -5.8007 -2.4739>
          <-32.0351 -0.2976 -5.7555>
          <-33.476 -3.204 -5.134>
          <-30.3048 0.3139 -5.9313>
          <-30.4667 1.5116 -5.7247>
          <-28.5444 -0.2517 -6.0792>
          <-27.5825 2.8934 -5.5506>
          <-28.7562 6.4022 -0.9764>
          <-27.9634 4.5577 -4.2609>
          <-27.6058 5.6734 -3.1402>
          <-24.1462 2.2636 -6.1264>
          <-21.3 6.4777 -2.0907>
          <-23.2373 3.1833 -5.8735>
          <-17.4033 7.102 -0.5075>
          <-17.2124 4.3352 -5.4335>
          <-15.2861 7.1696 -0.4814>
          <-13.9669 4.2074 -5.6777>
          <-12.0787 4.8344 -5.0607>
          <-15.5908 1.2181 -7.1461>
          <-12.3435 2.3924 -6.7009>
          <-14.504 -0.3653 -7.4949>
          <-19.9461 -2.0789 -7.215>
          <-26.3482 -1.4594 -6.3082>
          <-20.1525 -3.9287 -6.7242>
          <-32.9992 -5.4654 -3.7534>
          <-31.2142 -7.2369 -1.1402>
          <-30.8038 -7.22 -1.6274>
          <-24.8369 -7.6247 -3.0213>
          <-23.6994 -7.9491 -2.7765>
          <-34.8324 -4.6739 4.6653>
          <-28.909 -6.8835 3.1162>
          <-35.3353 -1.7467 6.2381>
          <-31.6759 -2.634 6.2938>
          <-29.391 -3.8955 6.0316>
          <-25.2847 -3.7396 6.774>
          <-26.0997 -5.3949 5.6253>
          <-22.7717 -3.9244 7.0596>
          <-23.0121 -7.8532 3.2641>
          <-21.0978 -7.7395 4.0527>
          <-17.9143 -8.2575 4.1122>
          <-18.727 -7.7058 4.6581>
          <-16.1989 -5.0068 7.2154>
          <-14.3788 -8.7664 4.3591>
          <-16.0909 -9.6877 -0.8094>
          <-13.6521 -8.5629 -4.3805>
          <-13.147 -6.6151 -6.3234>
          <-14.6004 -6.9521 -5.817>
          <-10.4529 -6.265 -6.7658>
          <-10.7211 2.0712 -6.7934>
          <-17.7582 -8.4185 -3.6294>
          <-19.0458 -7.4191 -4.7365>
          <-18.741 -9.0071 -2.3146>
          <-8.7887 6.7742 -1.0117>
          <-8.4771 6.7924 -0.2796>
         ]}}
       Random Seed=0
      )
      7=&SelectVectorFromArray(
       Array Sampling Mode=::NewEnumerator0
       Direct Array Index=?
       Vector Selection Array=DataInterface() {propertyValues {FloatData [
          <-36.6504 -4.9865 1.316>
          <-30.3522 -3.9671 2.5089>
          <-34.7497 -6.2872 -0.8354>
          <-40.8117 -5.4189 -2.463>
          <-39.8754 -4.09 -2.8633>
          <-42.2384 -1.2849 -3.2894>
          <-40.2927 -0.2261 -2.9791>
          <-42.3525 1.9227 -2.5994>
          <-41.2412 2.1143 -1.7411>
          <-43.2611 -0.947 1.941>
          <-38.4067 1.2407 -0.8803>
          <-36.5388 -0.4651 1.502>
          <-23.2592 -2.8896 3.9014>
          <-23.2672 -2.0878 3.5937>
          <-17.2849 -2.2164 4.2647>
          <-16.6444 -0.7571 3.6736>
          <-19.7972 0.074 2.6431>
          <-14.6199 -0.4554 3.5878>
          <-10.0803 -1.8816 4.3297>
          <-8.8059 -1.9156 4.3984>
          <-9.1353 -0.0888 3.6579>
          <-5.5533 -0.1884 4.339>
          <-5.3157 0.9905 3.4013>
          <-3.4804 2.6909 2.3011>
          <-19.0942 1.7091 0.9802>
          <-15.8616 2.7138 -0.8448>
          <-23.4926 1.2035 0.9615>
          <-27.3394 0.8839 0.8274>
          <-32.3598 0.0606 1.2939>
          <-33.6352 0.8031 0.3653>
          <-37.9287 -1.7895 -3.3113>
          <-37.1731 -4.512 -2.9498>
          <-34.1625 -3.3675 -3.8564>
          <-27.0257 -0.5989 -4.2826>
          <-20.3136 -1.2728 -5.3836>
          <-22.2689 -1.6683 -5.3524>
          <-16.4174 0.2456 -4.647>
          <-13.3737 2.2902 -2.7666>
          <-11.1442 0.2203 -4.6863>
          <-9.5197 1.2123 -3.9272>
          <-9.5326 -2.2198 -5.5523>
          <-8.5765 0.9564 -4.0555>
          <-7.7893 -0.5486 -4.8268>
          <-5.7422 0.8258 -3.9614>
          <-3.6052 2.3894 -2.7698>
          <-3.5006 -0.9231 -4.6634>
          <-3.6001 -1.0276 -4.7018>
          <-10.2414 -3.4009 -5.6663>
          <-14.2055 -6.03 -5.5476>
          <-15.5432 -5.4729 -5.8291>
          <-21.4313 -6.0088 -5.2477>
          <-23.6091 -5.1113 -5.2882>
          <-29.2364 -7.1827 -1.0153>
          <-21.2084 -7.8671 -3.3968>
          <-19.4903 -8.4616 1.203>
          <-20.8857 -4.9282 4.2634>
          <-21.8853 -6.4613 3.216>
          <-19.9542 -8.132 1.6186>
          <-9.0836 -5.6901 3.5698>
          <-8.1349 -3.9291 4.3056>
          <-11.3305 -8.1648 1.3184>
          <-5.3002 -5.0928 3.3004>
          <-3.7782 -6.1988 -1.8351>
          <-6.016 -7.0266 -0.2251>
          <-7.841 -7.4373 0.4462>
          <-7.5085 -7.4145 -1.3966>
          <-9.174 -7.679 -2.0157>
          <-10.0484 -7.6832 -2.6074>
          <-13.7797 -8.8082 -0.9597>
          <-15.3786 -8.9103 -1.636>
         ]}}
       Random Seed=0
      )
      8=&SelectVectorFromArray(
       Array Sampling Mode=::NewEnumerator0
       Direct Array Index=?
       Vector Selection Array=DataInterface() {propertyValues {FloatData [
          <-5.9266 13.5489 -0.2473>
          <-6.8976 11.7187 1.8326>
          <-7.539 4.5423 3.5078>
          <-0.7063 -3.4827 2.3737>
          <-0.0629 -4.8789 1.087>
          <-0.8342 -4.7285 0.3097>
          <-7.2896 -3.3843 -0.3304>
          <-5.6296 -3.5431 -0.3094>
          <-2.6074 -2.9502 2.165>
          <2.0859 -5.6234 1.1439>
          <3.1104 -4.3961 2.8837>
          <3.8766 -4.1616 2.9537>
          <6.2763 -4.5699 1.1885>
          <0.7665 -0.1466 3.2117>
          <3.0964 5.9115 1.3956>
          <3.0533 6.0904 -0.6729>
          <2.8137 5.8876 -0.9257>
          <1.4385 3.7865 -2.9098>
          <0.6073 5.3852 -2.3855>
          <0.1667 7.1412 -0.6964>
          <-1.7407 2.8637 -3.9241>
          <-2.748 4.4879 -4.2276>
          <-2.4932 6.8078 -3.4922>
          <-1.8045 8.9176 0.6268>
          <-4.4382 7.7141 3.5647>
          <-3.5507 8.7015 2.7676>
          <-4.4346 13.214 -0.9203>
          <-5.1901 5.391 -4.6229>
          <-2.8549 2.183 -3.9168>
          <-0.7588 -1.0657 -2.6823>
          <2.077 -5.7616 -0.4451>
          <3.3129 -3.0943 -2.6476>
          <3.7506 -5.1388 -1.3605>
          <3.8362 -4.3496 -2.0263>
          <6.5069 -3.7433 -1.414>
          <5.1739 -2.7596 -2.2854>
          <5.0182 -2.9568 -2.2841>
          <3.1879 -1.7312 -2.9091>
          <-8.1604 7.0177 2.2731>
          <-8.1548 6.3984 2.8781>
          <-8.2157 7.6895 -2.7115>
          <-7.9928 11.0469 0.5938>
          <-8.0223 10.8981 -0.7915>
          <-8.2092 4.351 -3.389>
          <-3.8047 2.5451 -0.4266>
          <-4.0525 2.6489 -1.3797>
          <-3.8053 2.8235 1.5774>
          <-2.7923 1.7828 0.1303>
         ]}}
       Random Seed=0
      )
      Index=$Particles.PartitionParticles.Partition
      Num=::NewEnumerator1
     )
     Float=0.9
    )
    Skeletal Mesh=$System.Skeletal Mesh
   )
   &ScaleSpriteSize(
    Uniform Curve Scale=&FNTimeOfDayScalar(Night Scale=1.25, Morning Scale=1, Evening Scale=1)
    Uniform Curve Sprite Scale=DataInterface() {propertyValues {Curve {keys [
        {interpMode=RCIM_Cubic, tangentMode=RCTM_User, time=0, value=0}
        {interpMode=RCIM_Cubic, tangentMode=RCTM_User, time=0.3, value=1}
        {tangentMode=RCTM_User, time=1, value=0}
       ]}}}
   ) {module=ScaleSpriteSize001}
   &ScaleColor(
    Scale Alpha=&FloatFromCurve(
     FloatCurve=DataInterface() {propertyValues {Curve {keys [
         {interpMode=RCIM_Cubic, tangentMode=RCTM_User, time=0, value=0}
         {interpMode=RCIM_Cubic, tangentMode=RCTM_User, time=0.3, value=1}
         {tangentMode=RCTM_User, time=1, value=0}
        ]}}}
     Scale Curve=&RandomRangeFloat(Minimum=0.25, Maximum=0.5)
    )
    Scale RGB=&VectorFromFloat(Value=&Multiply_Float(
      A=&RandomRangeFloat(Minimum=0.75, Maximum=1)
      B=&SelectFloatFromArray(Direct Array Index=$Engine.QualityLevel, Float Selection Array=DataInterface() {propertyValues {FloatData [2.5,2.5,2.5,1,1]}})
     ))
    ScaleA=true
    ScaleRGB=true
   ) {module=ScaleColor001}
   &ScaleColor(
    Color Value To Scale=$Particles.Color
    Scale Alpha=$System.TODAlpha
    Scale RGB=&MakeVectorFromLinearColorRGB(Color=$System.TODColor)
    ScaleA=true
    ScaleRGB=true
   )
   /SetVariables(
    Particles.CenterFlareColor:LinearColor=&Multiply_LinearColorByFloat(Linear Color=$Particles.Color, Float=1.2)
    Particles.CenterFlareSize:Vector2f=&Multiply_Vector2(A=$Particles.SpriteSize, B=&Vector2DFromFloat(Value=1))
    Particles.CoronaFlareColor:LinearColor=$Particles.Color
    Particles.CoronaFlareSize:Vector2f=&Multiply_Vector2(A=$Particles.SpriteSize, B=&Vector2DFromFloat(Value=0.75))
   ) {isSetParameters=true}
   &KillParticles(Kill Particles=&Boolean_AndOperation(
     Bool1=$User.bisFootHidden
     Bool2=&Boolean_OrOperation(
      Bool1=&SetBoolByIntComparison(A=$Particles.PartitionParticles.Partition, B=4, Comparison Type=::NewEnumerator4)
      Bool2=&Boolean_OrOperation(
       Bool1=&SetBoolByIntComparison(A=$Particles.PartitionParticles.Partition, B=5, Comparison Type=::NewEnumerator4)
       Bool2=&Boolean_OrOperation(
        Bool1=&SetBoolByIntComparison(A=$Particles.PartitionParticles.Partition, B=7, Comparison Type=::NewEnumerator4)
        Bool2=&SetBoolByIntComparison(A=$Particles.PartitionParticles.Partition, B=8, Comparison Type=::NewEnumerator4)
       )
      )
     )
    )) {enabled=false}
  ]
  rendererClasses [&NiagaraSpriteRendererProperties]
  renderers [
   {rendererClass=&NiagaraSpriteRendererProperties}
   {rendererClass=&NiagaraSpriteRendererProperties, rendererIndex=1}
  ]
 }}
summary {
 systemName=NS_Character_FloraBrisk_Idle
 emitters [{emitterName=Sparkles_Original, rendererClasses [&NiagaraSpriteRendererProperties]}]
 userVariables [
  User.bisFootHidden {type=NiagaraBool}
  User.bIsFrontEnd {type=NiagaraBool}
  User.bIsFrontEndPreview {type=NiagaraBool}
 ]
}
