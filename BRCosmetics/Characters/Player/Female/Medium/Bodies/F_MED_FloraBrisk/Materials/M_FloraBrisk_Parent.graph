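# Parent character material for the FloraBrisk female cosmetic body, built on the 2023 character PreFX and PostFX functions.
# Between them it chains thin film, optional custom crest, sequins, an optional emissive scan and fluffy fabric.
# The scan multiplies panned noise by the sequins' offset noise and a tint, masked by tech art red plus green, and sets Emissive Color.
# Red is thresholded for sequins and scan, while fabric gets the raw mask.
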Material /BRCosmetics/Characters/Player/Female/Medium/Bodies/F_MED_FloraBrisk/Materials/M_FloraBrisk_Parent
graph {
 expressions {
  MaterialExpressionAdd_1 {
   class=MaterialExpressionAdd
   inputs [
    A {from=MaterialExpressionComponentMask_1}
    B {from=MaterialExpressionComponentMask_2}
   ]
  }
  MaterialExpressionChannelMaskParameter_1 {
   class=MaterialExpressionChannelMaskParameter
   inputs [None {from=MaterialExpressionTextureSampleParameter2D_1, output=RGB}]
   properties {group=EmissiveScan, parameterName=EmissiveScan.Noise_Channel, sortPriority=14}
  }
  MaterialExpressionComponentMask_0 {class=MaterialExpressionComponentMask, inputs [None {from=MaterialExpressionVectorParameter_0, output=RGB}], properties {g=true, r=true}}
  MaterialExpressionComponentMask_1 {class=MaterialExpressionComponentMask, inputs [None {from=MaterialExpressionReroute_8, output=TechArtMask}], properties {r=true}}
  MaterialExpressionComponentMask_2 {class=MaterialExpressionComponentMask, inputs [None {from=MaterialExpressionReroute_8, output=TechArtMask}], properties {g=true}}
  MaterialExpressionComposite_0 {
   class=MaterialExpressionComposite
   inputs [
    IN_MA {from=MaterialExpressionMaterialFunctionCall_6, output=OUT_MA}
    UV {from="", output=""}
    OffsetNoise {from=MaterialExpressionMaterialFunctionCall_6, output=OUT_MA}
    TechArtMask {from=MaterialExpressionConvert_1, output=""}
   ]
   properties {subgraphName=MFW_EmissiveAnimation_TEMP}
  }
  MaterialExpressionConvert_1 {
   class=MaterialExpressionConvert
   inputs [
    None {from=MaterialExpressionStep_0}
    None {from=MaterialExpressionConvert_2}
    None {from=MaterialExpressionConvert_2}
   ]
   properties {
    convertInputs [
     {
      type=Scalar
      expressionInput {
       expression=&self:MaterialExpressionStep_0
       inputName=None
       mask=0
       maskA=0
       maskB=0
       maskG=0
       maskR=0
       outputIndex=0
      }
     }
     {
      type=Scalar
      expressionInput {
       expression=&self:MaterialExpressionConvert_2
       inputName=None
       mask=0
       maskA=0
       maskB=0
       maskG=0
       maskR=0
       outputIndex=1
      }
     }
     {
      type=Scalar
      expressionInput {
       expression=&self:MaterialExpressionConvert_2
       inputName=None
       mask=0
       maskA=0
       maskB=0
       maskG=0
       maskR=0
       outputIndex=2
      }
     }
    ]
    convertOutputs [{type=Vector3}]
   }
  }
  MaterialExpressionConvert_2 {
   class=MaterialExpressionConvert
   inputs [None {from=MaterialExpressionMaterialFunctionCall_1, output=Tech Art Mask}]
   properties {
    convertInputs [{
      type=Vector3
      expressionInput {
       expression=&self:MaterialExpressionMaterialFunctionCall_1
       inputName=None
       mask=0
       maskA=0
       maskB=0
       maskG=0
       maskR=0
       outputIndex=5
      }
     }]
    convertOutputs [{type=Scalar}, {type=Scalar}, {type=Scalar}]
   }
  }
  MaterialExpressionMaterialFunctionCall_0 {
   class=MaterialExpressionMaterialFunctionCall
   inputs [
    Material Attributes {from=MaterialExpressionMaterialFunctionCall_1, output=Result}
    Fresnel {from=MaterialExpressionMaterialFunctionCall_1, output=Result}
    Alpha {from=MaterialExpressionMaterialFunctionCall_1, output=Result}
   ]
   properties {materialFunction=/Game/Characters/Player/Female/Medium/Bodies/F_MED_SkeleProbe/Materials/MF_ThinFilmDuplicate_TechArt}
  }
  MaterialExpressionMaterialFunctionCall_1 {
   class=MaterialExpressionMaterialFunctionCall
   inputs [
    Overide Texture Set UVs {from=""}
    Override Tech Art Mask UVs {from=""}
    Override Basic Emissive UVs {from=""}
   ]
   properties {materialFunction=&MF_CharacterParent_2023_PreFX}
  }
  MaterialExpressionMaterialFunctionCall_2 {
   class=MaterialExpressionMaterialFunctionCall
   inputs [
    Material {from=MaterialExpressionMaterialFunctionCall_5, output=Result}
    M {from=MaterialExpressionMaterialFunctionCall_1, output=M}
    Pre Skinned Local Position {from=MaterialExpressionMaterialFunctionCall_1, output=M}
    Distance Blend {from=MaterialExpressionMaterialFunctionCall_1, output=M}
    Fresnel {from=MaterialExpressionMaterialFunctionCall_1, output=M}
   ]
   properties {materialFunction=&MF_CharacterParent_2023_PostFX}
  }
  MaterialExpressionMaterialFunctionCall_3 {
   class=MaterialExpressionMaterialFunctionCall
   inputs [Materials {from=MaterialExpressionMaterialFunctionCall_0, output=Result}]
   properties {materialFunction=/Game/Weapons/WeaponSkins/MaterialFunction/CustomCrests}
  }
  MaterialExpressionMaterialFunctionCall_5 {
   class=MaterialExpressionMaterialFunctionCall
   inputs [
    Material {from=MaterialExpressionReroute_9, output=OUT_MA}
    TechArtMask {from=MaterialExpressionReroute_1}
   ]
   properties {materialFunction=/Game/Global/Materials/Character_Parent_2023/FluffyFabric_Parent_2023/MF_FluffyFabric}
  }
  MaterialExpressionMaterialFunctionCall_6 {
   class=MaterialExpressionMaterialFunctionCall
   inputs [
    IN_MA {from=MaterialExpressionStaticSwitchParameter_0}
    UVs {from=""}
    TechArtMask {from=MaterialExpressionConvert_1}
    Fresnel {from=""}
   ]
   properties {materialFunction=/Game/Global/MaterialFunctions/Cosmetics/Sequins/MFW_Sequins}
  }
  MaterialExpressionMultiply_0 {
   class=MaterialExpressionMultiply
   inputs [
    A {from=MaterialExpressionTextureCoordinate_0}
    B {from=MaterialExpressionScalarParameter_0}
   ]
  }
  MaterialExpressionMultiply_1 {
   class=MaterialExpressionMultiply
   inputs [
    A {from=MaterialExpressionMultiply_2}
    B {from=MaterialExpressionVectorParameter_1, output=RGB}
   ]
  }
  MaterialExpressionMultiply_2 {
   class=MaterialExpressionMultiply
   inputs [
    A {from=MaterialExpressionReroute_7, output=OffsetNoise}
    B {from=MaterialExpressionPower_0}
   ]
  }
  MaterialExpressionMultiply_3 {
   class=MaterialExpressionMultiply
   inputs [
    A {from=MaterialExpressionMultiply_1}
    B {from=MaterialExpressionScalarParameter_1}
   ]
  }
  MaterialExpressionMultiply_4 {
   class=MaterialExpressionMultiply
   inputs [
    A {from=MaterialExpressionMultiply_3}
    B {from=MaterialExpressionAdd_1}
   ]
  }
  MaterialExpressionPanner_1 {
   class=MaterialExpressionPanner
   inputs [
    Coordinate {from=MaterialExpressionMultiply_0}
    Time {from=""}
    Speed {from=MaterialExpressionComponentMask_0}
   ]
   properties {speedX=0.05, speedY=0.05}
  }
  MaterialExpressionPinBase_0 {
   class=MaterialExpressionPinBase
   inputs [
    OUT_MA {from=MaterialExpressionStaticSwitchParameter_1, output=""}
    ScanPanner {from=MaterialExpressionPanner_1, output=""}
   ]
   properties {reroutePins [
     OUT_MA {expression=&self:MaterialExpressionReroute_9}
     ScanPanner {expression=&self:MaterialExpressionReroute_10}
    ]}
  }
  MaterialExpressionPinBase_1 {
   class=MaterialExpressionPinBase
   inputs []
   properties {reroutePins [
     IN_MA {expression=&self:MaterialExpressionReroute_5}
     UV {expression=&self:MaterialExpressionReroute_6}
     OffsetNoise {expression=&self:MaterialExpressionReroute_7}
     TechArtMask {expression=&self:MaterialExpressionReroute_8}
    ]}
  }
  MaterialExpressionPower_0 {
   class=MaterialExpressionPower
   inputs [
    Base {from=MaterialExpressionChannelMaskParameter_1, output=R}
    Exp {from=""}
   ]
  }
  MaterialExpressionReroute_0 {class=MaterialExpressionReroute, inputs [None {from=MaterialExpressionMaterialFunctionCall_1, output=Tech Art Mask}]}
  MaterialExpressionReroute_1 {class=MaterialExpressionReroute, inputs [None {from=MaterialExpressionReroute_2}]}
  MaterialExpressionReroute_10 {class=MaterialExpressionReroute, inputs [None {from=MaterialExpressionPanner_1}]}
  MaterialExpressionReroute_2 {class=MaterialExpressionReroute, inputs [None {from=MaterialExpressionReroute_0}]}
  MaterialExpressionReroute_5 {class=MaterialExpressionReroute, inputs [None {from=MaterialExpressionMaterialFunctionCall_6, output=OUT_MA}]}
  MaterialExpressionReroute_6 {class=MaterialExpressionReroute, inputs [None {from=""}]}
  MaterialExpressionReroute_7 {class=MaterialExpressionReroute, inputs [None {from=MaterialExpressionMaterialFunctionCall_6, output=Sequins_Offseted_Noise}]}
  MaterialExpressionReroute_8 {class=MaterialExpressionReroute, inputs [None {from=MaterialExpressionConvert_1}]}
  MaterialExpressionReroute_9 {class=MaterialExpressionReroute, inputs [None {from=MaterialExpressionStaticSwitchParameter_1}]}
  MaterialExpressionScalarParameter_0 {
   class=MaterialExpressionScalarParameter
   properties {defaultValue=2, group=EmissiveScan, parameterName=EmissiveScan.Noise_Tiling, sortPriority=14}
  }
  MaterialExpressionScalarParameter_1 {
   class=MaterialExpressionScalarParameter
   properties {defaultValue=10, group=EmissiveScan, parameterName=EmissiveScan.Color_Brightness, sortPriority=13}
  }
  MaterialExpressionSetMaterialAttributes_0 {class=MaterialExpressionSetMaterialAttributes, inputs [MaterialAttributes {from=MaterialExpressionMaterialFunctionCall_2, output=Result}]}
  MaterialExpressionSetMaterialAttributes_1 {
   class=MaterialExpressionSetMaterialAttributes
   inputs [
    MaterialAttributes {from=MaterialExpressionReroute_5, output=IN_MA}
    Emissive Color {from=MaterialExpressionMultiply_4, output=""}
   ]
   properties {attributeSetTypes [#1]}
  }
  MaterialExpressionStaticSwitchParameter_0 {
   class=MaterialExpressionStaticSwitchParameter
   inputs [
    True {from=MaterialExpressionMaterialFunctionCall_3, output=Result}
    False {from=MaterialExpressionMaterialFunctionCall_0, output=Result}
   ]
   properties {group=04_Tech Art, parameterName=useCustomCrest}
  }
  MaterialExpressionStaticSwitchParameter_1 {
   class=MaterialExpressionStaticSwitchParameter
   inputs [
    True {from=MaterialExpressionSetMaterialAttributes_1, output=""}
    False {from=MaterialExpressionReroute_5, output=IN_MA}
   ]
   properties {group=04_Tech Art, parameterName=UseEmissivePan}
  }
  MaterialExpressionStep_0 {
   class=MaterialExpressionStep
   inputs [
    Y {from="", output=""}
    X {from=MaterialExpressionConvert_2, output=""}
   ]
   properties {constX=0.1, constY=0.5}
  }
  MaterialExpressionTextureCoordinate_0 {class=MaterialExpressionTextureCoordinate}
  MaterialExpressionTextureSampleParameter2D_1 {
   class=MaterialExpressionTextureSampleParameter2D
   inputs [
    UVs {from=MaterialExpressionPanner_1, output=""}
    Apply View MipBias {from="", output=""}
   ]
   properties {
    group=EmissiveScan
    parameterName=EmissiveScan.NoiseMap
    samplerSource=SSM_Wrap_WorldGroupSettings
    sortPriority=14
    texture=/Game/Effects/Fort_Effects/Textures/Tile/T_ShieldCaustics_Tile_Chromatic
   }
  }
  MaterialExpressionVectorParameter_0 {
   class=MaterialExpressionVectorParameter
   properties {group=EmissiveScan, parameterName=EmissiveScan.Noise_Speed, sortPriority=14, defaultValue=<rgba 0.05 0.05 0 0>}
  }
  MaterialExpressionVectorParameter_1 {
   class=MaterialExpressionVectorParameter
   properties {group=EmissiveScan, parameterName=EmissiveScan.Color, defaultValue=<rgba 0.5742 0.228 1 1>}
  }
 }
 properties {
  bUseMaterialAttributes=true
  bUsedWithMeshDeformer=true
  bUsedWithMorphTargets=true
  bUsedWithSkeletalMesh=true
  pixelDepthOffsetMode=PDOM_Legacy
  shadingModel=MSM_FromMaterialExpression
  subsurfaceProfile=&SS_HeroSkin_02
  usedShadingModels=Default Lit | Subsurface | Clear Coat | Subsurface Profile | Cloth
 }
}
