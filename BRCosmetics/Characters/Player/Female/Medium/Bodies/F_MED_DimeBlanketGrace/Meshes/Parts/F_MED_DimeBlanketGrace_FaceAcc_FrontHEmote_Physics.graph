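# Physics asset for the front hair of the DimeBlanketGrace female medium face accessory, apparently an emote-specific variant.
# Left and right five-link capsule chains hang from small base bodies constrained to spine_05, with narrow twist and second-swing limits keeping bending mostly to one axis.
# Spine, clavicle, neck and right upper-arm capsules, linked by free angular constraints, supply body volumes around the hair; there is no left upper-arm body.
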PhysicsAsset /BRCosmetics/Characters/Player/Female/Medium/Bodies/F_MED_DimeBlanketGrace/Meshes/Parts/F_MED_DimeBlanketGrace_FaceAcc_FrontHEmote_Physics
bodies {
 base_hair_Front_L_01 [{shapeType=Capsule, length=0.505, radius=0.505, shapeName=base_hair_Front_L_01_capsule, rotation=<rot 0 -90 0>}]
 base_hair_Front_R_01 [{shapeType=Capsule, length=0.505, radius=0.505, shapeName=base_hair_Front_R_01_capsule, rotation=<rot 0 -90 0>}]
 clavicle_l [{
   shapeType=Capsule
   length=4.6779
   radius=6.1717
   shapeName=clavicle_l_capsule
   center=<6.3931 -1.0251 -2.2485>
   rotation=<rot 86.9643 -0.0014 -0.0017>
  }]
 clavicle_r [{
   shapeType=Capsule
   length=4.6779
   radius=6.1717
   shapeName=clavicle_l_capsule
   center=<-6.3931 1.0251 2.2485>
   rotation=<rot 86.9643 -0.0014 179.9983>
  }]
 dyn_hair_Front_L_01 [{
   shapeType=Capsule
   length=3.2414
   radius=3.1248
   shapeName=dyn_hair_Front_L_01_capsule
   center=<3.6241 -0.295 1.0737>
   rotation=<rot 66.8248 77.8284 0>
  }]
 dyn_hair_Front_L_02 [{
   shapeType=Capsule
   length=2.3714
   radius=4.2099
   shapeName=dyn_hair_Front_L_02_capsule
   center=<3.75 -2.1155 -0.974>
   rotation=<rot 0 -0.0001 -18.036>
  }]
 dyn_hair_Front_L_03 [{
   shapeType=Capsule
   length=1.1621
   radius=2.752
   shapeName=dyn_hair_Front_L_03_capsule
   center=<2.8836 -1.1608 -0.8091>
   rotation=<rot 0 0 -19.1821>
  }]
 dyn_hair_Front_L_04 [{
   shapeType=Capsule
   length=2.1588
   radius=2.7806
   shapeName=dyn_hair_Front_L_04_capsule
   center=<1.0434 -1.6576 -0.0792>
   rotation=<rot 65.1613 -93.9633 -90>
  }]
 dyn_hair_Front_L_05 [{shapeType=Capsule, length=2, radius=1, shapeName=dyn_hair_Front_L_05_capsule, center=<2 0 0>}]
 dyn_hair_Front_R_01 [{
   shapeType=Capsule
   length=3.8209
   radius=3.4147
   shapeName=dyn_hair_Front_R_01_capsule
   center=<3.9966 -0.1829 -0.3036>
   rotation=<rot 75.5123 -106.7777 -180>
  }]
 dyn_hair_Front_R_02 [{
   shapeType=Capsule
   length=4.4522
   radius=2.0307
   shapeName=dyn_hair_Front_R_02_capsule
   center=<2.8606 -0.4046 0.3138>
   rotation=<rot 0 -0.0001 6.1508>
  }]
 dyn_hair_Front_R_03 [{
   shapeType=Capsule
   length=4.6674
   radius=2.7077
   shapeName=dyn_hair_Front_R_03_capsule
   center=<3.0256 -1.4072 1.2122>
   rotation=<rot 0 0 21.8544>
  }]
 dyn_hair_Front_R_04 [{
   shapeType=Capsule
   length=3.3134
   radius=2.3991
   shapeName=dyn_hair_Front_R_04_capsule
   center=<2.2824 -1.8196 0.7693>
   rotation=<rot 90 180 -172.8626>
  }]
 dyn_hair_Front_R_05 [{shapeType=Capsule, length=2, radius=1, shapeName=dyn_hair_Front_R_05_capsule, center=<2 0 0>}]
 neck_01_mover [{shapeType=Capsule, length=38.813, radius=2, shapeName=neck_01_capsule, center=<0 -1.6521 0>}]
 spine_03 [{shapeType=Capsule, length=4.8535, radius=9.5365, shapeName=spine_03_capsule, center=<-2.0838 2.2176 0>}]
 spine_04 [{shapeType=Capsule, length=5.2699, radius=9.8801, shapeName=spine_04_capsule, center=<0 2.6192 0>}]
 spine_05 [{shapeType=Capsule, length=10.2422, radius=8.8577, shapeName=spine_05_capsule, center=<0.0464 -1.6321 0>}]
 upperarm_r [{shapeType=Capsule, length=16.5881, radius=2.0509, shapeName=upperarm_r_capsule, center=<-10.4447 0 0>, rotation=<rot -0.0001 -89.9999 90>}]
}
constraints [
 {
  bone1Name=spine_04
  bone2Name=spine_03
  swing1LimitDegrees=45
  swing1Motion=Free
  swing2LimitDegrees=45
  swing2Motion=Free
  twistLimitDegrees=45
  twistMotion=Free
 }
 {
  bone1Name=spine_05
  bone2Name=spine_04
  swing1LimitDegrees=45
  swing1Motion=Free
  swing2LimitDegrees=45
  swing2Motion=Free
  twistLimitDegrees=45
  twistMotion=Free
 }
 {
  bone1Name=clavicle_l
  bone2Name=spine_05
  swing1LimitDegrees=45
  swing1Motion=Free
  swing2LimitDegrees=45
  swing2Motion=Free
  twistLimitDegrees=45
  twistMotion=Free
 }
 {
  bone1Name=clavicle_r
  bone2Name=spine_05
  swing1LimitDegrees=45
  swing1Motion=Free
  swing2LimitDegrees=45
  swing2Motion=Free
  twistLimitDegrees=45
  twistMotion=Free
 }
 {
  bone1Name=base_hair_Front_L_01
  bone2Name=spine_05
  swing1LimitDegrees=45
  swing1Motion=Free
  swing2LimitDegrees=45
  swing2Motion=Free
  twistLimitDegrees=45
  twistMotion=Free
 }
 {bone1Name=dyn_hair_Front_L_01, bone2Name=base_hair_Front_L_01, swing1LimitDegrees=60, swing2LimitDegrees=10, twistLimitDegrees=1}
 {bone1Name=dyn_hair_Front_L_02, bone2Name=dyn_hair_Front_L_01, swing1LimitDegrees=45, swing2LimitDegrees=10, twistLimitDegrees=1}
 {bone1Name=dyn_hair_Front_L_03, bone2Name=dyn_hair_Front_L_02, swing1LimitDegrees=60, swing2LimitDegrees=10, twistLimitDegrees=1}
 {bone1Name=dyn_hair_Front_L_04, bone2Name=dyn_hair_Front_L_03, swing1LimitDegrees=60, swing2LimitDegrees=10, twistLimitDegrees=1}
 {bone1Name=dyn_hair_Front_L_05, bone2Name=dyn_hair_Front_L_04, swing1LimitDegrees=60, swing2LimitDegrees=10, twistLimitDegrees=1}
 {
  bone1Name=base_hair_Front_R_01
  bone2Name=spine_05
  swing1LimitDegrees=45
  swing1Motion=Free
  swing2LimitDegrees=45
  swing2Motion=Free
  twistLimitDegrees=45
  twistMotion=Free
 }
 {bone1Name=dyn_hair_Front_R_01, bone2Name=base_hair_Front_R_01, swing1LimitDegrees=45, swing2LimitDegrees=10, twistLimitDegrees=1}
 {bone1Name=dyn_hair_Front_R_02, bone2Name=dyn_hair_Front_R_01, swing1LimitDegrees=45, swing2LimitDegrees=10, twistLimitDegrees=1}
 {bone1Name=dyn_hair_Front_R_03, bone2Name=dyn_hair_Front_R_02, swing1LimitDegrees=45, swing2LimitDegrees=10, twistLimitDegrees=1}
 {bone1Name=dyn_hair_Front_R_04, bone2Name=dyn_hair_Front_R_03, swing1LimitDegrees=45, swing2LimitDegrees=10, twistLimitDegrees=1}
 {bone1Name=dyn_hair_Front_R_05, bone2Name=dyn_hair_Front_R_04, swing1LimitDegrees=45, swing2LimitDegrees=10, twistLimitDegrees=1}
 {
  bone1Name=neck_01_mover
  bone2Name=spine_05
  swing1LimitDegrees=45
  swing1Motion=Free
  swing2LimitDegrees=45
  swing2Motion=Free
  twistLimitDegrees=45
  twistMotion=Free
 }
 {
  bone1Name=upperarm_r
  bone2Name=clavicle_r
  swing1LimitDegrees=45
  swing1Motion=Free
  swing2LimitDegrees=45
  swing2Motion=Free
  twistLimitDegrees=45
  twistMotion=Free
 }
]
