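# Physics asset for the back hair on the DimeBlanketGrace female character's face accessory mesh, defining capsule bodies for secondary hair motion.
# Left, middle and right chains of four dynamic bones hang from small base capsules, with locked twist and swing limits generally looser below the root.
# Angularly free constraints tie both side chains to the middle one at the third segment.
# Spine, clavicle and upper-arm capsules supply body shapes around the hair.
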PhysicsAsset /BRCosmetics/Characters/Player/Female/Medium/Bodies/F_MED_DimeBlanketGrace/Meshes/Parts/F_MED_DimeBlanketGrace_FaceAcc_RangeBHair_Physics
bodies {
 base_backHair_L_01 [{shapeType=Capsule, length=0.505, radius=0.505, shapeName=base_backHair_L_01_capsule, rotation=<rot 0 -90 0>}]
 base_backHair_M_01 [{shapeType=Capsule, length=0.505, radius=0.505, shapeName=base_backHair_M_01_capsule, rotation=<rot 0 -90 0>}]
 base_backHair_R_01 [{shapeType=Capsule, length=0.505, radius=0.505, shapeName=base_backHair_R_01_capsule, rotation=<rot 0 -90 0>}]
 clavicle_l [{
   shapeType=Capsule
   length=9.326
   radius=5.6718
   shapeName=clavicle_l_capsule
   center=<6.3692 -0.7413 -2.2472>
   rotation=<rot 86.9638 -0.1986 -5.0246>
  }]
 clavicle_r [{
   shapeType=Capsule
   length=9.326
   radius=5.6718
   shapeName=clavicle_l_capsule
   center=<-6.3692 0.7413 2.2472>
   rotation=<rot 86.9638 -0.1986 -5.0246>
  }]
 dyn_backHair_L_01 [{shapeType=Capsule, length=7.2654, radius=3.7599, shapeName=dyn_backHair_L_01_capsule, center=<5.7166 0 0>, rotation=<rot 90 0 0>}]
 dyn_backHair_L_02 [{shapeType=Capsule, length=5.8556, radius=3.7812, shapeName=dyn_backHair_L_02_capsule, center=<5.1293 -0.8556 -1>, rotation=<rot 90 0 0>}]
 dyn_backHair_L_03 [{shapeType=Capsule, length=4.7527, radius=3.7273, shapeName=dyn_backHair_L_03_capsule, center=<4.4004 -0.8583 0>, rotation=<rot 90 0 0>}]
 dyn_backHair_L_04 [{shapeType=Capsule, length=0.843, radius=3.8369, shapeName=dyn_backHair_L_04_capsule, center=<2.7913 -0.5802 0>, rotation=<rot 90 0 0>}]
 dyn_backHair_M_01 [{shapeType=Capsule, length=7.2654, radius=3.7599, shapeName=dyn_backHair_M_01_capsule, center=<5.7166 0 0>, rotation=<rot 90 0 0>}]
 dyn_backHair_M_02 [{shapeType=Capsule, length=5.8556, radius=3.7812, shapeName=dyn_backHair_M_02_capsule, center=<5.1641 -0.8556 0>, rotation=<rot 90 0 0>}]
 dyn_backHair_M_03 [{shapeType=Capsule, length=4.7527, radius=3.7273, shapeName=dyn_backHair_M_03_capsule, center=<4.4909 -0.8583 0>, rotation=<rot 90 0 0>}]
 dyn_backHair_M_04 [{shapeType=Capsule, length=1.9831, radius=3.8369, shapeName=dyn_backHair_M_04_capsule, center=<3.256 0.3798 0>, rotation=<rot 90 0 0>}]
 dyn_backHair_R_01 [{shapeType=Capsule, length=7.2654, radius=3.7599, shapeName=dyn_backHair_R_01_capsule, center=<5.7166 0 0>, rotation=<rot 90 0 0>}]
 dyn_backHair_R_02 [{shapeType=Capsule, length=2.9444, radius=3.7812, shapeName=dyn_backHair_R_02_capsule, center=<4.8774 -0.8556 1>}]
 dyn_backHair_R_03 [{shapeType=Capsule, length=4.7527, radius=3.7273, shapeName=dyn_backHair_R_03_capsule, center=<4.4967 -0.3737 0>, rotation=<rot 90 0 0>}]
 dyn_backHair_R_04 [{shapeType=Capsule, length=0.6918, radius=3.8369, shapeName=dyn_backHair_R_04_capsule, center=<2.7659 -0.5802 0>, rotation=<rot 90 0 0>}]
 spine_03 [{shapeType=Capsule, length=19.5443, radius=9.3962, shapeName=spine_03_capsule, center=<-0.2117 0.9921 0>}]
 spine_04 [{shapeType=Capsule, length=14.3608, radius=9.8801, shapeName=spine_04_capsule, center=<0 2.2851 0>}]
 spine_05 [{shapeType=Capsule, length=10.8136, radius=8.2863, shapeName=spine_05_capsule, center=<0.0464 0.8639 0>}]
 upperarm_l [{
   shapeType=Capsule
   length=24.3425
   radius=5.6575
   shapeName=upperarm_l_capsule
   center=<12.6605 -0.1968 -0.0207>
   rotation=<rot -89.9062 179.9462 -179.9461>
  }]
 upperarm_r [{
   shapeType=Capsule
   length=24.3425
   radius=5.6575
   shapeName=upperarm_l_capsule
   center=<-12.6605 0.1968 0.0207>
   rotation=<rot -89.9062 179.9462 0.0539>
  }]
}
constraints [
 {
  bone1Name=dyn_backHair_L_01
  bone2Name=base_backHair_L_01
  swing1LimitDegrees=10
  swing2LimitDegrees=8
  twistLimitDegrees=1
  twistMotion=Locked
 }
 {
  bone1Name=dyn_backHair_L_02
  bone2Name=dyn_backHair_L_01
  swing1LimitDegrees=25
  swing2LimitDegrees=10
  twistLimitDegrees=1
  twistMotion=Locked
 }
 {
  bone1Name=dyn_backHair_L_03
  bone2Name=dyn_backHair_L_02
  swing1LimitDegrees=40
  swing2LimitDegrees=8
  twistLimitDegrees=1
  twistMotion=Locked
 }
 {
  bone1Name=dyn_backHair_L_04
  bone2Name=dyn_backHair_L_03
  swing1LimitDegrees=40
  swing2LimitDegrees=10
  twistLimitDegrees=1
  twistMotion=Locked
 }
 {
  bone1Name=dyn_backHair_M_01
  bone2Name=base_backHair_M_01
  swing1LimitDegrees=10
  swing2LimitDegrees=10
  twistLimitDegrees=1
  twistMotion=Locked
 }
 {
  bone1Name=dyn_backHair_M_02
  bone2Name=dyn_backHair_M_01
  swing1LimitDegrees=25
  swing2LimitDegrees=15
  twistLimitDegrees=1
  twistMotion=Locked
 }
 {
  bone1Name=dyn_backHair_M_03
  bone2Name=dyn_backHair_M_02
  swing1LimitDegrees=40
  swing2LimitDegrees=15
  twistLimitDegrees=1
  twistMotion=Locked
 }
 {
  bone1Name=dyn_backHair_M_04
  bone2Name=dyn_backHair_M_03
  swing1LimitDegrees=40
  swing2LimitDegrees=10
  twistLimitDegrees=1
  twistMotion=Locked
 }
 {
  bone1Name=dyn_backHair_R_01
  bone2Name=base_backHair_R_01
  swing1LimitDegrees=15
  swing2LimitDegrees=8
  twistLimitDegrees=1
  twistMotion=Locked
 }
 {
  bone1Name=dyn_backHair_R_02
  bone2Name=dyn_backHair_R_01
  swing1LimitDegrees=25
  swing2LimitDegrees=10
  twistLimitDegrees=1
  twistMotion=Locked
 }
 {
  bone1Name=dyn_backHair_R_03
  bone2Name=dyn_backHair_R_02
  swing1LimitDegrees=40
  swing2LimitDegrees=5
  twistLimitDegrees=1
  twistMotion=Locked
 }
 {
  bone1Name=dyn_backHair_R_04
  bone2Name=dyn_backHair_R_03
  swing1LimitDegrees=40
  swing2LimitDegrees=5
  twistLimitDegrees=1
  twistMotion=Locked
 }
 {
  bone1Name=dyn_backHair_R_03
  bone2Name=dyn_backHair_M_03
  swing1LimitDegrees=45
  swing1Motion=Free
  swing2LimitDegrees=45
  swing2Motion=Free
  twistLimitDegrees=45
  twistMotion=Free
 }
 {
  bone1Name=dyn_backHair_L_03
  bone2Name=dyn_backHair_M_03
  swing1LimitDegrees=45
  swing1Motion=Free
  swing2LimitDegrees=45
  swing2Motion=Free
  twistLimitDegrees=45
  twistMotion=Free
 }
]
