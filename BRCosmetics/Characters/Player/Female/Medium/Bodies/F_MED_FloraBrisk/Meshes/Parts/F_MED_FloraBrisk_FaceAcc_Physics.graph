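# Physics asset for the F_MED_FloraBrisk face accessory mesh, defining bodies and constraints for hair and top-hat bones.
# A four-link hair chain of small spheres hangs from base_hair, with tight twist and swing limits widening toward the tip.
# The top hat has narrow limits to the head.
# Head-to-spine, base-hair and back-hair joints swing and twist freely, with large zero-length capsules on spine_04 and both back-hair bones.
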PhysicsAsset /BRCosmetics/Characters/Player/Female/Medium/Bodies/F_MED_FloraBrisk/Meshes/Parts/F_MED_FloraBrisk_FaceAcc_Physics
bodies {
 base_hair [{shapeType=Sphere, radius=3, shapeName=base_hair_sphere}]
 dyn_backHair_M_02 [{shapeType=Capsule, length=0, radius=35.1522, shapeName=dyn_backHair_M_02_capsule, center=<6.3138 29.0346 -1.1869>, rotation=<rot 90 0 0>}]
 dyn_backHair_M_03 [{shapeType=Capsule, length=0, radius=35.1522, shapeName=dyn_backHair_M_02_capsule, center=<6.3138 29.0346 -1.1869>, rotation=<rot 90 0 0>}]
 dyn_hair_01 [{shapeType=Sphere, radius=3, shapeName=dyn_hair_01_sphere, center=<2.6046 0 0>}]
 dyn_hair_02 [{shapeType=Sphere, radius=3, shapeName=dyn_hair_02_sphere, center=<2.3027 0 0>}]
 dyn_hair_03 [{shapeType=Sphere, radius=3, shapeName=dyn_hair_03_sphere, center=<1.5725 0 0>}]
 dyn_hair_04 [{shapeType=Sphere, radius=3, shapeName=dyn_hair_04_sphere, center=<2.9211 0 0>}]
 dyn_top_hat [{shapeType=Sphere, radius=0.505, shapeName=dyn_top_hat_sphere, center=<0.411 0 0>}]
 head [{shapeType=Sphere, radius=4.7624, shapeName=head_sphere, center=<2.6606 -2.032 -1.9595>}]
 spine_04 [{shapeType=Capsule, length=0, radius=26, shapeName=spine_04_capsule, center=<0 11.6817 0>, rotation=<rot 0 -90 0>}]
}
constraints [
 {bone1Name=dyn_hair_01, bone2Name=base_hair, swing1LimitDegrees=20, swing2LimitDegrees=20, twistLimitDegrees=1}
 {bone1Name=dyn_hair_02, bone2Name=dyn_hair_01, swing1LimitDegrees=25, swing2LimitDegrees=25, twistLimitDegrees=1}
 {bone1Name=dyn_hair_03, bone2Name=dyn_hair_02, swing1LimitDegrees=30, swing2LimitDegrees=30, twistLimitDegrees=1}
 {bone1Name=dyn_hair_04, bone2Name=dyn_hair_03, swing1LimitDegrees=45, swing2LimitDegrees=45, twistLimitDegrees=1}
 {
  bone1Name=base_hair
  bone2Name=head
  swing1LimitDegrees=45
  swing1Motion=Free
  swing2LimitDegrees=45
  swing2Motion=Free
  twistLimitDegrees=45
  twistMotion=Free
 }
 {bone1Name=dyn_top_hat, bone2Name=head, swing1LimitDegrees=5, swing2LimitDegrees=20, twistLimitDegrees=1}
 {
  bone1Name=head
  bone2Name=spine_04
  swing1LimitDegrees=45
  swing1Motion=Free
  swing2LimitDegrees=45
  swing2Motion=Free
  twistLimitDegrees=45
  twistMotion=Free
 }
 {
  bone1Name=dyn_backHair_M_02
  bone2Name=head
  swing1LimitDegrees=45
  swing1Motion=Free
  swing2LimitDegrees=45
  swing2Motion=Free
  twistLimitDegrees=45
  twistMotion=Free
 }
 {
  bone1Name=dyn_backHair_M_03
  bone2Name=dyn_backHair_M_02
  swing1LimitDegrees=45
  swing1Motion=Free
  swing2LimitDegrees=45
  swing2Motion=Free
  twistLimitDegrees=45
  twistMotion=Free
 }
]
