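# Post-process style AnimBP for the FloraBrisk face accessory that drives front and back hair over the pose copied from its parent mesh.
# A Control Rig feeds cached RigidBody stages, mostly borrowing DimeBlanketGrace assets, selected by KnotRange mesh name and emote state.
# With backbling or cape, constraints pin back hair to spine_05 instead of simulating it.
# A final head-relative simulation precedes corrective bone offsets, one scaled by head yaw when rigid bodies are off.
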AnimBlueprint /BRCosmetics/Characters/Player/Female/Medium/Bodies/F_MED_FloraBrisk/Meshes/Parts/F_MED_FloraBrisk_FaceAcc_AnimBP
target_skeleton=&/F_MED_FloraBrisk_FaceAcc_Skeleton
graphs [
 AnimGraph {nodes [
   Component To Local()
   Local To Component(LocalPose=Copy Pose From Mesh() {bUseAttachedParent=true})
   n1 = RigidBody!(ComponentPose=Local To Component(LocalPose=Use cached pose 'FrontHair'()), Alpha, OverrideWorldGravity:Vector="0,0,-980", ExternalForce) -> (Pose) {
    bClampLinearTranslationLimitToRefPose=true
    bOverrideWorldGravity=true
    overridePhysicsAsset=/BRCosmetics/Characters/Player/Female/Medium/Bodies/F_MED_DimeBlanketGrace/Meshes/Parts/F_MED_DimeBlanketGrace_FaceAcc_RangeBHairEm_Physics
    simulationSpace=BaseBoneSpace
    alphaBoolBlend {blendOption=Linear}
    baseBoneRef {boneName=neck_02}
    componentAppliedLinearAccClamp=<3000 3000 3000>
    simSpaceSettings {
     maxAngularAcceleration=16
     maxAngularVelocity=8
     maxLinearAcceleration=700
     maxLinearVelocity=700
     worldAlpha=1
     externalAngularVelocity {z=0.1}
    }
   }
   n2 = RigidBody(ComponentPose=Local To Component(LocalPose=Use cached pose 'FrontHair'()), OverrideWorldGravity="0,0,0") {
    bClampLinearTranslationLimitToRefPose=true
    bOverrideWorldGravity=true
    overridePhysicsAsset=/BRCosmetics/Characters/Player/Female/Medium/Bodies/F_MED_DimeBlanketGrace/Meshes/Parts/F_MED_DimeBlanketGrace_FaceAcc_RangeBHair_Physics
    simulationSpace=BaseBoneSpace
    alphaBoolBlend {blendOption=Linear}
    baseBoneRef {boneName=neck_02}
    componentAppliedLinearAccClamp=<0 0 0>
    simSpaceSettings {
     maxAngularAcceleration=16
     maxAngularVelocity=8
     maxLinearAcceleration=700
     maxLinearVelocity=700
     worldAlpha=1
     externalAngularVelocity {z=0.1}
    }
   }
   n3 = RigidBody!(
    ComponentPose=Local To Component(LocalPose=Use cached pose 'FrontHair'())
    bAlphaBoolEnabled:bool
    OverrideWorldGravity="0,0,0"
    ExternalForce
   ) -> (Pose) {
    alphaInputType=Bool
    bClampLinearTranslationLimitToRefPose=true
    bOverrideWorldGravity=true
    overridePhysicsAsset=&/F_MED_FloraBrisk_FaceAcc_BackHair_Physics
    simulationSpace=BaseBoneSpace
    alphaBoolBlend {blendOption=Linear}
    baseBoneRef {boneName=neck_02}
    componentAppliedLinearAccClamp=<0 0 0>
    simSpaceSettings {
     maxAngularAcceleration=16
     maxAngularVelocity=8
     maxLinearAcceleration=700
     maxLinearVelocity=700
     worldAlpha=0.9
     externalAngularVelocity {z=0.1}
    }
   }
   n4 = Blend Poses by bool(
    BlendPose_0=Blend Poses by bool(BlendPose_0=Component To Local(ComponentPose=@n1), BlendPose_1=Component To Local(ComponentPose=@n2), bActiveValue=Get bIsPlayingEmote()) {blendType=Linear}
    BlendPose_1=Component To Local(ComponentPose=@n3)
    bActiveValue=Get KnotRange Alpha() -> (KnotRange Alpha:bool)
   ) {blendType=Linear}
   n5 = Use cached pose 'Backpack or Cape Setup'()
   n6 = Get is Wearing Backbling or Cape() -> (is Wearing Backbling or Cape:bool)
   n7 = RigidBody!(
    ComponentPose=Local To Component(LocalPose=Blend Poses by bool(BlendPose_0=@n5, BlendPose_1=@n4, bActiveValue=@n6) {blendType=Linear})
    Alpha:float=0.7
    OverrideWorldGravity="0,0,0"
    ExternalForce
   ) -> (Pose) {
    bClampLinearTranslationLimitToRefPose=true
    bOverrideWorldGravity=true
    overridePhysicsAsset=&/F_MED_FloraBrisk_FaceAcc_Physics
    simulationSpace=BaseBoneSpace
    alphaBoolBlend {blendOption=Linear}
    baseBoneRef {boneName=head}
    simSpaceSettings {dampingAlpha=0, maxAngularAcceleration=70, maxAngularVelocity=8, maxLinearAcceleration=1100, maxLinearVelocity=700, worldAlpha=1}
   }
   n8 = "Transform (Modify) Bone - Bone: base_hair"(
    ComponentPose=@n7
    Alpha=Get Rigid_Body_off() -> (Rigid_Body_off:double)
    Translation="-4.131,-0.000,-5.075"
    Rotation="60,0,0"
    Scale="1,1,1"
   ) {
    rotationMode=BMM_Additive
    rotationSpace=BCS_BoneSpace
    translationMode=BMM_Additive
    translationSpace=BCS_BoneSpace
    boneToModify {boneName=base_hair}
    rotation {pitch=60}
    translation {x=-4.131, z=-5.075}
   }
   n9 = "Transform (Modify) Bone - Bone: Back_Hair_corr"!(ComponentPose=@n8, bAlphaBoolEnabled:bool=True, Translation="-3.000,-8.581,-0.010", Rotation="0,0,0", Scale="1,1,1") -> (Pose) {
    alphaInputType=Bool
    rotationMode=BMM_Additive
    rotationSpace=BCS_BoneSpace
    translationMode=BMM_Additive
    translationSpace=BCS_BoneSpace
    boneToModify {boneName=Back_Hair_corr}
    translation=<-3 -8.581 -0.01>
   }
   Output Pose(Result=Component To Local(ComponentPose=@n9)) {layerGroup=None}
   n10 = Control Rig!(
    Source=Copy Pose From Mesh() {bUseAttachedParent=true}
    Alpha
    is Sky Floating:bool=Get bIsSkydiveFloating()
    is Skydive Dive:bool=Get bIsSkydiveDiveMode()
    is Parachute Open:bool=Get bIsParachuteOpened()
    is Playing Emote:bool=Get bIsPlayingEmote()
    Range Variant Equipped:bool=Get KnotRange Alpha()
    Is Wearing Backpack or Cape Alpha:double=Get is Wearing Backbling or Cape Alpha() -> (is Wearing Backbling or Cape Alpha:double)
   ) -> (Pose) {
    alphaBoolBlend {blendOption=Linear}
    controlRigAssetReference {blueprintRigClass=/BRCosmetics/Characters/Player/Female/Medium/Bodies/F_MED_DimeBlanketGrace/Meshes/Parts/F_MED_DimeBlanketGrace_FaceAcc_CtrlRig.F_MED_DimeBlanketGrace_FaceAcc_CtrlRig_C}
   }
   n11 = "Transform (Modify) Bone - Bone: neck_01_mover"(ComponentPose=Local To Component(LocalPose=@n10), Alpha=1, Translation="0,-10,0", Rotation="0,0,0", Scale="1,1,1") {
    translationMode=BMM_Additive
    translationSpace=BCS_BoneSpace
    alphaBoolBlend {blendOption=Linear}
    boneToModify {boneName=neck_01_mover}
    translation {y=-10}
   }
   Save cached pose 'Start Pose'(Pose=Component To Local(ComponentPose=@n11)) {cacheName=Start Pose}
   n12 = "Constraint - base_backHair_L_01 (PP)"!(ComponentPose=Local To Component(LocalPose=Use cached pose 'FrontHair'()), Alpha=1, ConstraintWeights_0=1, ConstraintWeights_1:float=0) -> (Pose) {
    alphaBoolBlend {blendOption=Linear}
    boneToModify {boneName=base_backHair_L_01}
    constraintSetup [{targetBone {boneName=spine_05}}, {targetBone {boneName=head}}]
    constraintWeights [1,1]
   }
   n13 = "Constraint - base_backHair_M_01 (P)"(ComponentPose=@n12, Alpha=1, ConstraintWeights_0=1) {
    alphaBoolBlend {blendOption=Linear}
    boneToModify {boneName=base_backHair_M_01}
    constraintSetup [{targetBone {boneName=spine_05}}]
    constraintWeights [1]
   }
   n14 = "Constraint - base_backHair_R_01 (PT)"!(ComponentPose=@n13, Alpha=1, ConstraintWeights_0=1, ConstraintWeights_1:float=0) -> (Pose) {
    alphaBoolBlend {blendOption=Linear}
    boneToModify {boneName=base_backHair_R_01}
    constraintSetup [{targetBone {boneName=spine_05}}, {transformType=Translation, targetBone {boneName=neck_02}}]
    constraintWeights [1,1]
   }
   n15 = "Constraint - dyn_backHair_L_02 (P)"(ComponentPose=@n14, Alpha=1, ConstraintWeights_0=1) {
    alphaBoolBlend {blendOption=Linear}
    boneToModify {boneName=dyn_backHair_L_02}
    constraintSetup [{targetBone {boneName=spine_05}}]
    constraintWeights [1]
   }
   n16 = "Constraint - dyn_backHair_M_02 (P)"(ComponentPose=@n15, Alpha=1, ConstraintWeights_0=1) {
    alphaBoolBlend {blendOption=Linear}
    boneToModify {boneName=dyn_backHair_M_02}
    constraintSetup [{targetBone {boneName=spine_05}}]
    constraintWeights [1]
   }
   n17 = "Constraint - dyn_backHair_R_02 (P)"(ComponentPose=@n16, Alpha=1, ConstraintWeights_0=1) {
    alphaBoolBlend {blendOption=Linear}
    boneToModify {boneName=dyn_backHair_R_02}
    constraintSetup [{targetBone {boneName=spine_05}}]
    constraintWeights [1]
   }
   n18 = "Constraint - dyn_backHair_L_03 (P)"(ComponentPose=@n17, Alpha=1, ConstraintWeights_0=1) {
    alphaBoolBlend {blendOption=Linear}
    boneToModify {boneName=dyn_backHair_L_03}
    constraintSetup [{targetBone {boneName=spine_05}}]
    constraintWeights [1]
   }
   n19 = "Constraint - dyn_backHair_M_03 (P)"(ComponentPose=@n18, Alpha=1, ConstraintWeights_0=1) {
    alphaBoolBlend {blendOption=Linear}
    boneToModify {boneName=dyn_backHair_M_03}
    constraintSetup [{targetBone {boneName=spine_05}}]
    constraintWeights [1]
   }
   n20 = "Constraint - dyn_backHair_R_03 (P)"(ComponentPose=@n19, Alpha=1, ConstraintWeights_0=1) {
    alphaBoolBlend {blendOption=Linear}
    boneToModify {boneName=dyn_backHair_R_03}
    constraintSetup [{targetBone {boneName=spine_05}}]
    constraintWeights [1]
   }
   n21 = "Constraint - dyn_backHair_L_04 (P)"(ComponentPose=@n20, Alpha=1, ConstraintWeights_0=1) {
    alphaBoolBlend {blendOption=Linear}
    boneToModify {boneName=dyn_backHair_L_04}
    constraintSetup [{targetBone {boneName=spine_05}}]
    constraintWeights [1]
   }
   n22 = "Constraint - dyn_backHair_M_04 (P)"(ComponentPose=@n21, Alpha=1, ConstraintWeights_0=1) {
    alphaBoolBlend {blendOption=Linear}
    boneToModify {boneName=dyn_backHair_M_04}
    constraintSetup [{targetBone {boneName=spine_05}}]
    constraintWeights [1]
   }
   n23 = "Constraint - dyn_backHair_R_04 (P)"(ComponentPose=@n22, Alpha=1, ConstraintWeights_0=1) {
    alphaBoolBlend {blendOption=Linear}
    boneToModify {boneName=dyn_backHair_R_04}
    constraintSetup [{targetBone {boneName=spine_05}}]
    constraintWeights [1]
   }
   n24 = "Transform (Modify) Bone - Bone: dyn_backHair_L_03"(ComponentPose=@n23, Alpha=1, Translation="0,1,0", Rotation="0,0,0", Scale="1,1,1") {
    translationMode=BMM_Additive
    translationSpace=BCS_BoneSpace
    alphaBoolBlend {blendOption=Linear}
    boneToModify {boneName=dyn_backHair_L_03}
    translation {y=1}
   }
   n25 = "Transform (Modify) Bone - Bone: dyn_backHair_M_03"(ComponentPose=@n24, Alpha=1, Translation="0,1,0", Rotation="0,0,0", Scale="1,1,1") {
    translationMode=BMM_Additive
    translationSpace=BCS_BoneSpace
    alphaBoolBlend {blendOption=Linear}
    boneToModify {boneName=dyn_backHair_M_03}
    translation {y=1}
   }
   n26 = "Transform (Modify) Bone - Bone: dyn_backHair_M_03"(ComponentPose=@n25, Alpha=1, Translation="0,1,0", Rotation="0,0,0", Scale="1,1,1") {
    translationMode=BMM_Additive
    translationSpace=BCS_BoneSpace
    alphaBoolBlend {blendOption=Linear}
    boneToModify {boneName=dyn_backHair_M_03}
    translation {y=1}
   }
   Save cached pose 'Backpack or Cape Setup'(Pose=Component To Local(ComponentPose=@n26)) {cacheName=Backpack or Cape Setup}
   n27 = RigidBody(ComponentPose=Local To Component(LocalPose=Use cached pose 'Start Pose'()), OverrideWorldGravity="0,0,0") {
    bClampLinearTranslationLimitToRefPose=true
    bOverrideWorldGravity=true
    overridePhysicsAsset=/BRCosmetics/Characters/Player/Female/Medium/Bodies/F_MED_DimeBlanketGrace/Meshes/Parts/F_MED_DimeBlanketGrace_FaceAcc_FrontHEmote_Physics
    simulationSpace=BaseBoneSpace
    alphaBoolBlend {blendOption=Linear}
    baseBoneRef {boneName=neck_02}
    componentAppliedLinearAccClamp=<0 0 0>
    simSpaceSettings {maxAngularAcceleration=16, maxAngularVelocity=8, maxLinearAcceleration=700, maxLinearVelocity=700, velocityScaleZ=0.7, worldAlpha=1}
   }
   n28 = Look At - Bone: dyn_hair_Front_R_01(ComponentPose=@n27, Alpha=1) {
    bUseLookUpAxis=true
    alphaBoolBlend {blendOption=Linear}
    boneToModify {boneName=dyn_hair_Front_R_01}
    lookAtLocation=<-11.3993 14.594 6.0411>
    lookAtTarget {boneReference {boneName=spine_05}}
    lookAt_Axis {axis=<1 0>}
    lookUp_Axis {axis=<0 1>}
   }
   n29 = Component To Local(ComponentPose=@n28)
   n30 = RigidBody(ComponentPose=Local To Component(LocalPose=Use cached pose 'Start Pose'()), OverrideWorldGravity="0,0,0") {
    bClampLinearTranslationLimitToRefPose=true
    bOverrideWorldGravity=true
    overridePhysicsAsset=/BRCosmetics/Characters/Player/Female/Medium/Bodies/F_MED_DimeBlanketGrace/Meshes/Parts/F_MED_DimeBlanketGrace_FaceAcc_FrontHair_Physics
    simulationSpace=BaseBoneSpace
    alphaBoolBlend {blendOption=Linear}
    baseBoneRef {boneName=neck_02}
    componentAppliedLinearAccClamp=<0 0 0>
    simSpaceSettings {maxAngularAcceleration=16, maxAngularVelocity=8, maxLinearAcceleration=700, maxLinearVelocity=700, velocityScaleZ=0.7, worldAlpha=1}
   }
   n31 = Component To Local(ComponentPose=@n30)
   n32 = Get bIsPlayingEmote()
   n33 = RigidBody(ComponentPose=Local To Component(LocalPose=Use cached pose 'Start Pose'()), OverrideWorldGravity="0,0,0") {
    bClampLinearTranslationLimitToRefPose=true
    bOverrideWorldGravity=true
    overridePhysicsAsset=/BRCosmetics/Characters/Player/Female/Medium/Bodies/F_MED_DimeBlanketGrace/Meshes/Parts/F_MED_DimeBlanketGrace_FaceAcc_RFHairEmotes_Physics
    simulationSpace=BaseBoneSpace
    alphaBoolBlend {blendOption=Linear}
    baseBoneRef {boneName=neck_02}
    componentAppliedLinearAccClamp=<0 0 0>
    simSpaceSettings {maxAngularAcceleration=16, maxAngularVelocity=8, maxLinearAcceleration=700, maxLinearVelocity=700, velocityScaleZ=0.7, worldAlpha=1}
   }
   n34 = Look At - Bone: dyn_hair_Front_R_01(ComponentPose=@n33, Alpha=1) {
    bUseLookUpAxis=true
    alphaBoolBlend {blendOption=Linear}
    boneToModify {boneName=dyn_hair_Front_R_01}
    lookAtLocation=<-11.3993 14.594 6.0411>
    lookAtTarget {boneReference {boneName=spine_05}}
    lookAt_Axis {axis=<1 0>}
    lookUp_Axis {axis=<0 1>}
   }
   n35 = Component To Local(ComponentPose=@n34)
   n36 = RigidBody(ComponentPose=Local To Component(LocalPose=Use cached pose 'Start Pose'()), OverrideWorldGravity="0,0,0") {
    bClampLinearTranslationLimitToRefPose=true
    bOverrideWorldGravity=true
    overridePhysicsAsset=/BRCosmetics/Characters/Player/Female/Medium/Bodies/F_MED_DimeBlanketGrace/Meshes/Parts/F_MED_DimeBlanketGrace_FaceAcc_RangeFHair_Physics
    simulationSpace=BaseBoneSpace
    alphaBoolBlend {blendOption=Linear}
    baseBoneRef {boneName=neck_02}
    componentAppliedLinearAccClamp=<0 0 0>
    simSpaceSettings {maxAngularAcceleration=16, maxAngularVelocity=8, maxLinearAcceleration=700, maxLinearVelocity=700, velocityScaleZ=0.7, worldAlpha=1}
   }
   n37 = Component To Local(ComponentPose=@n36)
   Save cached pose 'FrontHair'(Pose=Blend Poses by bool(
     BlendPose_0=Blend Poses by bool(BlendPose_0=@n35, BlendPose_1=@n37, bActiveValue=@n32) {blendType=Linear}
     BlendPose_1=Blend Poses by bool(BlendPose_0=@n29, BlendPose_1=@n31, bActiveValue=@n32) {blendType=Linear}
     bActiveValue=Get KnotRange Alpha()
    ) {blendType=Linear}) {cacheName=FrontHair}
  ]}
 EventGraph {nodes [
   Event BlueprintInitializeAnimation()
   > Set Skel Mesh(Skel Mesh=GetOwnerSkeletalMeshComponent())
   Event OnCharacterCustomizationLoaded()
   n1 = CheckForBackbling()
   > Set is Wearing Backbling(is Wearing Backbling:bool=AND Boolean!(A=@n1, B=NOT Boolean(A=@n1.bIsBackblingACape), C:bool=NOT Boolean(A=@n1.bIsBackblingLong)) -> (ReturnValue)) -> (Output_Get:bool)
   > Set is Wearing Cape(is Wearing Cape:bool=OR Boolean(A=@n1.bIsBackblingACape, B=@n1.bIsBackblingLong)) -> (Output_Get:bool)
   n2 = OR Boolean(A=Get is Wearing Backbling() -> (is Wearing Backbling:bool), B=Get is Wearing Cape() -> (is Wearing Cape:bool))
   > Set is Wearing Backbling or Cape(is Wearing Backbling or Cape:bool=@n2) -> (Output_Get:bool)
   > Set is Wearing Backbling or Cape Alpha(is Wearing Backbling or Cape Alpha:double="To Float (Boolean)"(InBool=@n2)) -> (Output_Get:double)
   > n3 = Cast To FortPlayerPawn(Object=TryGetPawnOwner())
   n4 = GetSkeletalMeshForPartType(self=@n3)
   > Set KnotRange Alpha(KnotRange Alpha:bool=MatchesWildcard(SourceString=GetDisplayName(Object=@n4), Wildcard=*F_MED_DimeBlanketKnot_Range*, SearchCase=CaseSensitive)) -> (Output_Get:bool)
   Event BlueprintUpdateAnimation()
   > n5 = Sequence!(execute) -> (then_0, then_1, then_2:Exec)
   > n6 = Gravity Override State Function DEPRECATED(CharacterPartAnimInstance=Self-Reference(), GameTimeInSeconds=GetGameTimeInSeconds(), GravityOverrideParams=Get Gravity Override Params())
   > Set GravityOverride1(GravityOverride1:Vector=@n6) -> (Output_Get:Vector)
   n7 = Gravity Override State Function DEPRECATED(
    execute=@n5.then_1
    CharacterPartAnimInstance=Self-Reference()
    GameTimeInSeconds=GetGameTimeInSeconds()
    GravityOverrideParams=Get Gravity Override Params()
   )
   > Set GravityOverride(GravityOverride=@n7)
   > Set is Skyfloating(is Skyfloating:bool=OR Boolean(A=Get bIsSkydiveFloating(), B=Get bIsSkydiveDiveMode())) -> (Output_Get:bool)
   n8 = Get bIsJumping()
   n9 = Get bIsFalling()
   > Set is Jumping Alpha(is Jumping Alpha:double="To Float (Boolean)"(InBool=OR Boolean(A=@n8, B=@n9))) -> (Output_Get:double)
   n10 = Get bRigidBodyEnabled()
   > Set RBAN Alpha(RBAN Alpha:double="To Float (Boolean)"(InBool=NOT Boolean(A=@n10))) -> (Output_Get:double)
   > Set Neck Mover In/Out(Neck Mover In/Out:bool=OR Boolean(A=Get bIsInFrontEnd(), B=Get bIsOnGround() -> (bIsOnGround:bool))) -> (Output_Get:bool)
   > Set Head Rotation Right(Head Rotation Right:double=GetAlphaFromReferenceTransformDelta(
     SkeletalMeshComponent=Get Skel Mesh()
     BoneName=head
     BaseName=spine_05
     DeltaComponentType=EulerX
     InputRange=(X=0,Y=-52)
     TwistAxis="1,0,0"
    )) -> (Output_Get:double)
   > Set is Playing Emote Alpha(is Playing Emote Alpha:double="To Float (Boolean)"(InBool=Get bIsPlayingEmote())) -> (Output_Get:double)
   > Set is Playing Emote Hair Corrective(is Playing Emote Hair Corrective:double=float * float(A=Get Head Rotation Right() -> (Head Rotation Right:double), B=Get is Playing Emote Alpha() -> (is Playing Emote Alpha:double)) {is=title K2Node_PromotableOperator}) -> (Output_Get:double)
   n11 = Get bRigidBodyEnabled()
   n12 = Get bIsPlayingEmote()
   > Set Idle_FE(Idle_FE:bool=AND Boolean(A=NOT Boolean(A=@n11), B=Get bIsInFrontEnd(), C=NOT Boolean(A=@n12))) -> (Output_Get:bool)
   > n13 = Branch(Condition=Get bRigidBodyEnabled())
   > Set Rigid_Body_off(Rigid_Body_off:double=0) -> (Output_Get:double)
   n14 = GetOwnerSkeletalMeshComponent()
   Set Rigid_Body_off(
    execute=@n13.else
    Rigid_Body_off=MapRangeClamped(
     Value=GetDeltaTransformFromRefPose!(self=@n14, BoneName=head, BaseName=spine_05) -> (
      ReturnValue_Location:Vector
      ReturnValue_Rotation_Roll:float
      ReturnValue_Rotation_Pitch:float
      ReturnValue_Rotation_Yaw:float
      ReturnValue_Scale:Vector
     ).ReturnValue_Rotation_Yaw
     InRangeA=-40
     InRangeB=-100
    )
   )
   GetAlphaFromReferenceTransformDelta(
    SkeletalMeshComponent=Get Skel Mesh()
    BoneName=head
    BaseName=spine_05
    DeltaComponentType=EulerY
    InputRange=(X=15,Y=60)
    TwistAxis="1,0,0"
   )
  ]}
]
variables [
 Gravity Override Params {type=GravityOverrideParamsStruct}
 GravityOverride {type=Vector}
 GravityOverride1 {type=Vector}
 Head Rotation Right {type=number}
 Idle_FE {type=boolean}
 is Jumping Alpha {type=number}
 is Playing Emote Alpha {type=number}
 is Playing Emote Hair Corrective {type=number}
 is Skyfloating {type=boolean}
 is Wearing Backbling {type=boolean}
 is Wearing Backbling or Cape {type=boolean}
 is Wearing Backbling or Cape Alpha {type=number}
 is Wearing Cape {type=boolean}
 KnotRange Alpha {type=boolean}
 Neck Mover In/Out {type=boolean}
 RBAN Alpha {type=number}
 Rigid_Body_off {type=number}
 Skel Mesh {type=SkeletalMeshComponent}
]
